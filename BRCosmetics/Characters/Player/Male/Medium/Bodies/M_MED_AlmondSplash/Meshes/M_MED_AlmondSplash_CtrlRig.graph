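# Control Rig for the M_MED_AlmondSplash male player cosmetic body, correcting deformation and posing its hood, jacket, belt and candy cane bones.
# A shared DeformRig function runs first, then Deform Jacket uses spherical pose readers on head, upper arms and thighs to weight bone offsets.
# The isCloth flag switches thigh-driven jacket offsets to belt and candy cane offsets.
# Pack, rail-grinding and cape-plus-crouch flags select further fixes; some nodes are unwired.
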ControlRigBlueprint /BRCosmetics/Characters/Player/Male/Medium/Bodies/M_MED_AlmondSplash/Meshes/M_MED_AlmondSplash_CtrlRig
graphs [
 New Function_ContainedGraph {nodes [
   n1 = RigVMFunctionEntryNode()
   n2 = &RigVMFunction_Sequence(ExecuteContext=@n1.ExecuteContext)
   RigVMFunction_Sequence = RigVMAggregateNode(ExecuteContext=@n2.A)
   n3 = &RigUnit_SphericalPoseReader(
    ActiveRegionScaleFactors=(PositiveWidth=1,NegativeWidth=1,PositiveHeight=1,NegativeHeight=1)
    ActiveRegionSize=0.125
    Debug=(bDrawDebug=false,bDraw2D=False,bDrawLocalAxes=False,DebugScale=25,DebugSegments=20,DebugThickness=0.25)
    DriverAxis=(X=0,Y=1,Z=0)
    DriverItem=(Type=Bone,Name="head")
    FalloffRegionScaleFactors=(PositiveWidth=1,NegativeWidth=1,PositiveHeight=1,NegativeHeight=1)
    FalloffSize=0.3
    FlipHeightScaling=False
    FlipWidthScaling=False
    OptionalParentItem=(Type=Bone,Name="None")
    RotationOffset=(X=90,Y=0,Z=-50)
    ExecutePin=@RigVMFunction_Sequence.A
   )
   > &RigUnit_ModifyTransforms(
    ItemToModify="((Item=(Type=Bone,Name=\"dyn_hood_c_MSTR\"),Transform=(Rotation=(X=0.0103,Y=0.1412,Z=-0.0887,W=0.986),Translation=(X=0,Y=0,Z=0),Scale3D=(X=1,Y=1,Z=1))),(Item=(Type=Bone,Name=\"dyn_hood_mid_MSTR_R\"),Transform=(Rotation=(X=0,Y=0.1527,Z=0,W=0.9883),Translation=( ...(truncated)"
    WeightMinimum=0
    Weight=@n3.OutputParam
   )
   > n4 = &RigUnit_SphericalPoseReader(
    ActiveRegionScaleFactors=(PositiveWidth=1,NegativeWidth=1,PositiveHeight=1,NegativeHeight=1)
    ActiveRegionSize=0.15
    Debug=(bDrawDebug=false,bDraw2D=False,bDrawLocalAxes=False,DebugScale=25,DebugSegments=20,DebugThickness=0.25)
    DriverAxis=(X=1,Y=0,Z=0)
    DriverItem=(Type=Bone,Name="upperarm_l")
    FalloffRegionScaleFactors=(PositiveWidth=1,NegativeWidth=1,PositiveHeight=1,NegativeHeight=1)
    FalloffSize=0.4
    FlipHeightScaling=False
    FlipWidthScaling=False
    OptionalParentItem=(Type=Bone,Name="None")
    RotationOffset=(X=17.2691,Y=-12.0456,Z=28.5116)
   )
   > &RigUnit_ModifyTransforms(
    ItemToModify="((Item=(Type=Bone,Name=\"dyn_hood_fr_MSTR_L\"),Transform=(Rotation=(X=-0.1671,Y=-0.3268,Z=-0.0505,W=0.9288),Translation=(X=0,Y=0,Z=0),Scale3D=(X=0.5,Y=1,Z=1))),(Item=(Type=Bone,Name=\"dyn_hood_mid_MSTR_L\"),Transform=(Rotation=(X=0.5261,Y=-0.0256,Z=0.061,W=0.8 ...(truncated)"
    WeightMinimum=0
    Weight=@n4.OutputParam
   )
   > n5 = &RigUnit_SphericalPoseReader(
    ActiveRegionScaleFactors=(PositiveWidth=1,NegativeWidth=1,PositiveHeight=1,NegativeHeight=1)
    ActiveRegionSize=0.1
    Debug=(bDrawDebug=false,bDraw2D=False,bDrawLocalAxes=False,DebugScale=25,DebugSegments=20,DebugThickness=0.25)
    DriverAxis=(X=-1,Y=0,Z=0)
    DriverItem=(Type=Bone,Name="upperarm_r")
    FalloffRegionScaleFactors=(PositiveWidth=1,NegativeWidth=1,PositiveHeight=1,NegativeHeight=1)
    FalloffSize=0.5
    FlipHeightScaling=False
    FlipWidthScaling=False
    OptionalParentItem=(Type=Bone,Name="None")
    RotationOffset=(X=-114.8296,Y=5.0511,Z=35.1428)
   )
   n6 = &RigVMFunction_ControlFlowBranch(ExecuteContext=@n5.ExecutePin, Condition=RigVMVariableNode() {variable=isWearingPack}.Value)
   &RigUnit_ModifyTransforms(
    ItemToModify="((Item=(Type=Bone,Name=\"dyn_hood_fr_MSTR_R\"),Transform=(Rotation=(X=0.0598,Y=-0.0748,Z=-0.0406,W=0.9946),Translation=(X=0,Y=0,Z=0),Scale3D=(X=0.4,Y=1,Z=1))),(Item=(Type=Bone,Name=\"dyn_hood_fr_r_02\"),Transform=(Rotation=(X=0,Y=0,Z=0,W=1),Translation=(X=0,Y= ...(truncated)"
    WeightMinimum=0
    ExecutePin=@n6.False
    Weight=@n5.OutputParam
   )
   &RigUnit_ModifyTransforms(
    ItemToModify="((Item=(Type=Bone,Name=\"dyn_hood_fr_MSTR_R\"),Transform=(Rotation=(X=0.0598,Y=-0.0748,Z=-0.0406,W=0.9946),Translation=(X=0,Y=0,Z=0),Scale3D=(X=0.65,Y=1,Z=1))),(Item=(Type=Bone,Name=\"dyn_hood_fr_r_02\"),Transform=(Rotation=(X=0,Y=0,Z=0,W=1),Translation=(X=0,Y ...(truncated)"
    WeightMinimum=0
    ExecutePin=@n6.True
    Weight=@n5.OutputParam
   )
   n7 = &RigUnit_SphericalPoseReader(
    ActiveRegionScaleFactors=(PositiveWidth=2,NegativeWidth=1.5,PositiveHeight=1.75,NegativeHeight=1)
    ActiveRegionSize=0.1
    Debug=(bDrawDebug=false,bDraw2D=False,bDrawLocalAxes=False,DebugScale=25,DebugSegments=20,DebugThickness=0.25)
    DriverAxis=(X=-1,Y=0,Z=0)
    DriverItem=(Type=Bone,Name="thigh_l")
    FalloffRegionScaleFactors=(PositiveWidth=1,NegativeWidth=1.2,PositiveHeight=1,NegativeHeight=0.5)
    FalloffSize=0.3
    FlipHeightScaling=False
    FlipWidthScaling=False
    OptionalParentItem=(Type=Bone,Name="pelvis")
    RotationOffset=(X=130,Y=0,Z=-15)
    ExecutePin=@RigVMFunction_Sequence.B
   )
   > RigVMVariableNode(Value=@n7.OutputParam) {variable=spr_l_thigh_up_45}
   > n8 = &RigUnit_SphericalPoseReader(
    ActiveRegionScaleFactors=(PositiveWidth=3,NegativeWidth=0.8,PositiveHeight=2,NegativeHeight=2)
    ActiveRegionSize=0.1
    Debug=(bDrawDebug=false,bDraw2D=False,bDrawLocalAxes=False,DebugScale=25,DebugSegments=20,DebugThickness=0.25)
    DriverAxis=(X=-1,Y=0,Z=0)
    DriverItem=(Type=Bone,Name="thigh_l")
    FalloffRegionScaleFactors=(PositiveWidth=2,NegativeWidth=1.35,PositiveHeight=2,NegativeHeight=2)
    FalloffSize=0.26
    FlipHeightScaling=false
    FlipWidthScaling=false
    OptionalParentItem=(Type=Bone,Name="pelvis")
    RotationOffset=(X=180,Y=-32,Z=0)
   )
   > RigVMVariableNode(Value=@n8.OutputParam) {variable=spr_l_thigh_up_side}
   > n9 = &RigUnit_SphericalPoseReader(
    ActiveRegionScaleFactors=(PositiveWidth=3,NegativeWidth=2,PositiveHeight=2,NegativeHeight=2)
    ActiveRegionSize=0.1
    Debug=(bDrawDebug=false,bDraw2D=False,bDrawLocalAxes=False,DebugScale=25,DebugSegments=20,DebugThickness=0.25)
    DriverAxis=(X=1,Y=0,Z=0)
    DriverItem=(Type=Bone,Name="thigh_r")
    FalloffRegionScaleFactors=(PositiveWidth=2,NegativeWidth=2,PositiveHeight=2,NegativeHeight=2)
    FalloffSize=0.295
    FlipHeightScaling=False
    FlipWidthScaling=true
    OptionalParentItem=(Type=Bone,Name="pelvis")
    RotationOffset=(X=0,Y=-20,Z=0)
   )
   > RigVMVariableNode(Value=@n9.OutputParam) {variable=spr_r_thigh_up_45}
   > n10 = &RigUnit_SphericalPoseReader(
    ActiveRegionScaleFactors=(PositiveWidth=4,NegativeWidth=3,PositiveHeight=2.5,NegativeHeight=2)
    ActiveRegionSize=0.1
    Debug=(bDrawDebug=false,bDraw2D=False,bDrawLocalAxes=False,DebugScale=25,DebugSegments=20,DebugThickness=0.25)
    DriverAxis=(X=1,Y=0,Z=0)
    DriverItem=(Type=Bone,Name="thigh_r")
    FalloffRegionScaleFactors=(PositiveWidth=1,NegativeWidth=4,PositiveHeight=1,NegativeHeight=0.5)
    FalloffSize=0.36
    FlipHeightScaling=False
    FlipWidthScaling=true
    OptionalParentItem=(Type=Bone,Name="pelvis")
    RotationOffset=(X=-35,Y=0,Z=-35)
   )
   RigVMFunction_AlphaInterp = &RigVMFunction_AlphaInterp(InterpSpeedDecreasing=10, InterpSpeedIncreasing=10, Scale=1, bClampResult=False, bInterpResult=True, Value=@n10.OutputParam)
   > RigVMVariableNode(Value=@RigVMFunction_AlphaInterp.Result) {variable=spr_r_thigh_up_side}
   RigVMFunction_ControlFlowBranch = &RigVMFunction_ControlFlowBranch(ExecuteContext=@RigVMFunction_Sequence.C, Condition=RigVMVariableNode() {variable=isCloth}.Value)
   &RigUnit_ModifyTransforms(
    ItemToModify="((Item=(Type=Bone,Name=\"dyn_jacket_fr_MSTR_L\"),Transform=(Rotation=(X=-0.0234,Y=-0.059,Z=-0.3679,W=0.9277),Translation=(X=0,Y=0,Z=0),Scale3D=(X=1,Y=1,Z=1))),(Item=(Type=Bone,Name=\"dyn_jacket_MSTR_L\"),Transform=(Rotation=(X=-0.2211,Y=-0.0312,Z=-0.3065,W=0.9 ...(truncated)"
    WeightMinimum=0
    ExecutePin=@RigVMFunction_ControlFlowBranch.False
    Weight=RigVMVariableNode() {variable=spr_l_thigh_up_45}.Value
   )
   n11 = RigVMVariableNode() {variable=spr_r_thigh_up_side}
   > &RigUnit_ModifyTransforms(
    ItemToModify="((Item=(Type=Bone,Name=\"dyn_jacket_MSTR_R\"),Transform=(Rotation=(X=-0.5761,Y=-0.1153,Z=-0.0825,W=0.805),Translation=(X=0,Y=0,Z=0),Scale3D=(X=1,Y=1,Z=1))),(Item=(Type=Bone,Name=\"dyn_jacket_fr_MSTR_R\"),Transform=(Rotation=(X=0,Y=0,Z=0.304,W=0.9527),Translati ...(truncated)"
    WeightMinimum=0
    Weight=@n11.Value
   )
   n12 = &RigUnit_ModifyTransforms(
    ItemToModify="((Item=(Type=Bone,Name=\"dyn_candy_cane\"),Transform=(Rotation=(X=-0.0688,Y=0.0016,Z=0.0225,W=0.9974),Translation=(X=2.5284,Y=0.6312,Z=-1.6902),Scale3D=(X=0.75,Y=0.5,Z=0.75))),(Item=(Type=Bone,Name=\"dyn_belt\"),Transform=(Rotation=(X=0.0229,Y=0,Z=0,W=0.9997), ...(truncated)"
    WeightMinimum=0
    ExecutePin=@RigVMFunction_ControlFlowBranch.True
    Weight=@n11.Value
   )
   n13 = &RigVMFunction_ControlFlowBranch(ExecuteContext=@n12.ExecutePin, Condition=RigVMVariableNode() {variable=isRailGrinding}.Value)
   &RigUnit_ModifyTransforms(
    ItemToModify="((Item=(Type=Bone,Name=\"dyn_candy_cane\"),Transform=(Rotation=(X=-0.0688,Y=0.0016,Z=0.0225,W=0.9974),Translation=(X=2.5284,Y=0.6312,Z=-1.6902),Scale3D=(X=0.75,Y=0.5,Z=0.75))),(Item=(Type=Bone,Name=\"dyn_belt\"),Transform=(Rotation=(X=0.0229,Y=0,Z=0,W=0.9997), ...(truncated)"
    Weight=1
    WeightMinimum=0
    ExecutePin=@n13.True
   )
   RigVMFunctionReturnNode(ExecuteContext=@n2.B)
   &RigUnit_ModifyTransforms(
    ExecutePin=()
    ItemToModify="((Item=(Type=Bone,Name=\"dyn_candy_cane\"),Transform=(Rotation=(X=-0.0116,Y=0.0063,Z=0.0437,W=0.999),Translation=(X=3.8484,Y=0.6765,Z=-0.7083),Scale3D=(X=1,Y=1,Z=1))),(Item=(Type=Bone,Name=\"dyn_belt\"),Transform=(Rotation=(X=0,Y=0,Z=0.0904,W=0.9959),Translati ...(truncated)"
    Weight=1
    WeightMinimum=0
   )
   RigVMVariableNode() {variable=spr_r_thigh_up_45}
  ]}
 RigVMFunctionLibrary {nodes [Deform Jacket = RigVMCollapseNode(ExecuteContext=())]}
 RigVMFunction_Sequence_2_ContainedGraph {nodes [
   n1 = RigVMFunctionEntryNode()
   RigVMFunction_Sequence = &RigVMFunction_Sequence(ExecuteContext=@n1.ExecuteContext)
   n2 = &RigVMFunction_Sequence(ExecuteContext=@RigVMFunction_Sequence.B)
   n3 = &RigVMFunction_Sequence(ExecuteContext=@n2.B)
   RigVMFunctionReturnNode(A=@RigVMFunction_Sequence.A, B=@n2.A, C=@n3.A, D=@n3.B)
  ]}
 RigVMModel {nodes [
   RigUnit_BeginExecution = &RigUnit_BeginExecution()
   RigVMFunction_Sequence = &RigVMFunction_Sequence(ExecuteContext=@RigUnit_BeginExecution.ExecutePin)
   FN_DeformRig_FNC = RigVMFunctionReferenceNode(
    DeformRig_ Curve="(deformRig_l_upperarm_up_alpha_72_#1=False,deformRig_r_upperarm_up_alpha_73_#2=False,deformRig_l_upperarm_fr_alpha_74_#3=False,deformRig_r_upperarm_fr_alpha_75_CB51B8 ...(truncated)"
    DeformRig_Bool="(upperarm_1_#4=True,elbow_3_#5=True,wrist_7_#6=True,groin_8_#7=True,glute_17_#8=True,knee_9_40C408324CA261 ...(truncated)"
    DeformRig_Debug=(upperarm_debug_18_#4=False,wrist_debug_20_#6=False,lat_debug_26_#9=False,groinGlute_debug_27_#7=False)
    DeformRig_Value="(upperarm_up_18_#10=(Rotation=(X=0,Y=0,Z=0,W=1),Translation=(X=0,Y=0,Z=8),Scale3D=(X=0.7,Y=1,Z=1)),upperarm_fwd_fr_19_#11=(Rotation=(X=0,Y=0,Z=0,W=1),Translation=(X=0,Y=10,Z=0),Scale3D=(X=0.5,Y=1,Z= ...(truncated)"
    ExecuteContext=@RigVMFunction_Sequence.A
   )
   > Deform Jacket = RigVMFunctionReferenceNode()
   RigVMFunction_MathBoolAnd = &RigVMFunction_MathBoolAnd(A=RigVMVariableNode() {variable=isCloth}.Value, B=RigVMVariableNode() {variable=isWearingCape}.Value)
   RigVMFunction_ControlFlowBranch = &RigVMFunction_ControlFlowBranch(ExecuteContext=@"Deform Jacket".ExecuteContext, Condition=@RigVMFunction_MathBoolAnd.Result)
   n1 = &RigVMFunction_ControlFlowBranch(ExecuteContext=@RigVMFunction_ControlFlowBranch.True, Condition=RigVMVariableNode() {variable=isCrouching}.Value)
   &RigUnit_ModifyTransforms(
    ItemToModify="((Item=(Type=Bone,Name=\"dyn_jacket_bk_MSTR\"),Transform=(Rotation=(X=0,Y=0.0691,Z=0,W=0.9976),Translation=(X=0,Y=0,Z=0),Scale3D=(X=1,Y=1,Z=1))),(Item=(Type=Bone,Name=\"dyn_jacket_mid_MSTR_L\"),Transform=(Rotation=(X=0,Y=0.1571,Z=0,W=0.9876),Translation=(X=0,Y ...(truncated)"
    Weight=1
    WeightMinimum=0
    ExecutePin=@n1.True
   )
   n2 = &RigVMFunction_ControlFlowBranch(ExecuteContext=(), Condition=RigVMVariableNode() {variable=isCrouchMoving}.Value)
   &RigUnit_ModifyTransforms(
    ItemToModify="((Item=(Type=Bone,Name=\"dyn_jacket_bk_03\"),Transform=(Rotation=(X=0,Y=0,Z=0,W=1),Translation=(X=0,Y=0,Z=-2),Scale3D=(X=1,Y=1,Z=1))),(Item=(Type=Bone,Name=\"dyn_jacket_mid_r_04\"),Transform=(Rotation=(X=0,Y=0,Z=0,W=1),Translation=(X=0,Y=0,Z=-2),Scale3D=(X=1,Y ...(truncated)"
    Weight=1
    WeightMinimum=0
    ExecutePin=@n2.True
   )
   &RigUnit_ModifyTransforms(
    ExecutePin=()
    ItemToModify="((Item=(Type=Bone,Name=\"dyn_jacket_mid_r_04\"),Transform=(Rotation=(X=0,Y=0,Z=0,W=1),Translation=(X=0,Y=0,Z=-6),Scale3D=(X=1,Y=1,Z=1))),(Item=(Type=Bone,Name=\"dyn_jacket_mid_l_04\"),Transform=(Rotation=(X=0,Y=0,Z=0,W=1),Translation=(X=0,Y=0,Z=-6),Scale3D=(X= ...(truncated)"
    Weight=1
    WeightMinimum=0
   )
   &RigUnit_ModifyTransforms(
    ExecutePin=()
    ItemToModify="((Item=(Type=Bone,Name=\"dyn_jacket_mid_r_04\"),Transform=(Rotation=(X=0,Y=0,Z=0,W=1),Translation=(X=0,Y=0,Z=-6),Scale3D=(X=1,Y=1,Z=1))),(Item=(Type=Bone,Name=\"dyn_jacket_mid_l_04\"),Transform=(Rotation=(X=0,Y=0,Z=0,W=1),Translation=(X=0,Y=0,Z=-6),Scale3D=(X= ...(truncated)"
    Weight=1
    WeightMinimum=0
   )
  ]}
]
hierarchy [
 root {loc=<0 0 0>, rot=<0 0 0 1>}
 pelvis {parent=root, loc=<0 -0.7071 93.1191>, rot=<0.0127 0.707 -0.0127 -0.707>}
 spine_01 {parent=pelvis, loc=<6.7494 -0.6388 0>, rot=<0 0 0.034 -0.9994>}
 spine_02 {parent=spine_01, loc=<7.252 -0.035 0>, rot=<0 0 0.0238 -0.9997>}
 spine_03 {parent=spine_02, loc=<6.0106 0 0>, rot=<0 0 0.0149 -0.9999>}
 spine_04 {parent=spine_03, loc=<8.1023 0 0>, rot=<0 0 -0.0132 -0.9999>}
 spine_05 {parent=spine_04, loc=<10.0142 0 0>, rot=<0 0 -0.0015 -1>}
 attach_backpack {parent=spine_05, loc=<0.4492 -14.3575 0.1707>, rot=<-0.0056 -0.7065 0.0096 -0.7076>}
 neck_01 {parent=spine_05, loc=<11.3138 0.1529 0>, rot=<0 0 -0.2003 -0.9797>}
 neck_02 {parent=neck_01, loc=<4.5627 0 0>, rot=<0 0 0.0504 -0.9987>}
 head {parent=neck_02, loc=<5.3914 0 0>, rot=<0 0 0.115 -0.9934>}
 clavicle_l {parent=spine_05, loc=<4.3575 -0.1466 -2.3223>, rot=<0.0465 -0.7289 -0.009 -0.6829>}
 upperarm_l {parent=clavicle_l, loc=<15.5111 0 0>, rot=<0.0226 -0.3234 -0.0088 -0.9459>}
 lowerarm_l {parent=upperarm_l, loc=<26.6514 0 0>, rot=<0.0391 -0.0099 -0.2453 -0.9686>}
 hand_l {parent=lowerarm_l, loc=<22.3491 0 0>, rot=<-0.7081 -0.0507 0.0975 -0.6975>}
 thumb_01_l {parent=hand_l, loc=<1.3118 -0.3191 -2.459>, rot=<0.6546 -0.4042 0.0664 -0.6354>}
 thumb_02_l {parent=thumb_01_l, loc=<5.9336 -0.0448 0.0401>, rot=<-0.0037 0.0213 0.298 -0.9543>}
 thumb_03_l {parent=thumb_02_l, loc=<2.2223 0.3682 -0.0907>, rot=<0.0118 -0.013 0.0066 -0.9998>}
 index_metacarpal_l {parent=hand_l, loc=<1.6535 0.0183 -2.96>, rot=<0.0158 -0.0124 -0.0173 -0.9996>}
 index_01_l {parent=index_metacarpal_l, loc=<8.3008 0.0357 -0.0524>, rot=<-0.0044 0.0485 0.0924 -0.9945>}
 index_02_l {parent=index_01_l, loc=<3.3587 -0.3004 0.0452>, rot=<-0.002 0.0027 0.3225 -0.9466>}
 index_03_l {parent=index_02_l, loc=<2.5578 -0.123 -0.0252>, rot=<-0.002 0.0067 0.2558 -0.9667>}
 middle_metacarpal_l {parent=hand_l, loc=<2.1856 -0.1725 -0.5984>, rot=<0.006 0.0168 -0.0329 -0.9993>}
 middle_01_l {parent=middle_metacarpal_l, loc=<7.3275 -0.2447 0.0129>, rot=<-0.0208 0.0042 0.0858 -0.9961>}
 middle_02_l {parent=middle_01_l, loc=<4.0123 0.1401 0.1175>, rot=<-0.0218 0.0258 0.3512 -0.9357>}
 middle_03_l {parent=middle_02_l, loc=<2.6077 0.0534 0.1373>, rot=<0.0082 0.0188 0.1708 -0.9851>}
 ring_metacarpal_l {parent=hand_l, loc=<2.2074 -0.3625 1.3638>, rot=<-0.0284 0.0722 -0.0458 -0.9959>}
 ring_01_l {parent=ring_metacarpal_l, loc=<7.0295 -0.6197 -0.1446>, rot=<-0.0489 -0.0252 0.1892 -0.9804>}
 ring_02_l {parent=ring_01_l, loc=<3.6174 0.4647 -0.1027>, rot=<0 0 0.2806 -0.9598>}
 ring_03_l {parent=ring_02_l, loc=<2.4152 0.114 0.0199>, rot=<0 0 0.1189 -0.9929>}
 pinky_metacarpal_l {parent=hand_l, loc=<1.7616 -0.9753 3.3814>, rot=<-0.0397 0.0814 -0.069 -0.9935>}
 pinky_01_l {parent=pinky_metacarpal_l, loc=<7.1709 -0.4831 -0.1094>, rot=<-0.0654 -0.0035 0.2242 -0.9723>}
 pinky_02_l {parent=pinky_01_l, loc=<2.8274 0.0893 -0.1043>, rot=<0.0037 0.0011 0.2854 -0.9584>}
 pinky_03_l {parent=pinky_02_l, loc=<2.2046 0.0307 -0.0078>, rot=<-0.0002 0.0098 0.1445 -0.9895>}
 deform_wrist_dn_base_l {parent=hand_l, loc=<0 -2.2342 0>, rot=<0 0 0 -1>}
 deform_wrist_dn_l {parent=deform_wrist_dn_base_l, loc=<0 0 0>, rot=<0 0 0 1>}
 deform_wrist_up_base_l {parent=hand_l, loc=<0 2.2682 0>, rot=<0 0 0 -1>}
 deform_wrist_up_l {parent=deform_wrist_up_base_l, loc=<0 0 0>, rot=<0 0 0 1>}
 lowerarm_twist_02_l {parent=lowerarm_l, loc=<9.6199 0 0>, rot=<0 0 0 -1>}
 lowerarm_twist_01_l {parent=lowerarm_l, loc=<17.8549 0 0>, rot=<0 0 0 -1>}
 deform_elbow_in_base_l {parent=lowerarm_l, loc=<0.2112 5.259 0.8274>, rot=<0 0 0 -1>}
 deform_elbow_in_l {parent=deform_elbow_in_base_l, loc=<0 0 0>, rot=<0 0 0 1>}
 deform_elbow_out_base_l {parent=lowerarm_l, loc=<0 -4 0>, rot=<0 0 0 -1>}
 deform_elbow_out_l {parent=deform_elbow_out_base_l, loc=<0 0 0>, rot=<0 0 0 1>}
 upperarm_twist_01_l {parent=upperarm_l, loc=<0.0041 0 0>, rot=<0 0 0 -1>}
 deform_upperarm_bk_base_l {parent=upperarm_twist_01_l, loc=<1.0813 -7.2193 -1.9453>, rot=<0 0 0 1>}
 deform_upperarm_bk_l {parent=deform_upperarm_bk_base_l, loc=<0 0 0>, rot=<0 0 0 1>}
 deform_upperarm_fr_base_l {parent=upperarm_twist_01_l, loc=<-0.3691 7.0506 -2.765>, rot=<0 0 0 1>}
 deform_upperarm_fr_l {parent=deform_upperarm_fr_base_l, loc=<0 0 0>, rot=<0 0 0 1>}
 deform_upperarm_up_base_l {parent=upperarm_twist_01_l, loc=<-3.2216 -1.1391 3.3311>, rot=<0 0 0 1>}
 deform_upperarm_up_l {parent=deform_upperarm_up_base_l, loc=<0 0 0>, rot=<0 0 0 1>}
 upperarm_twist_02_l {parent=upperarm_l, loc=<13.7941 0 0>, rot=<0 0 0 -1>}
 attach_cape {parent=spine_05, loc=<9.6694 -12.5295 0.0483>, rot=<0.0654 -0.7039 -0.0612 -0.7046>}
 clavicle_r {parent=spine_05, loc=<4.3575 -0.1466 2.3223>, rot=<0.7289 0.0465 -0.6829 0.009>}
 upperarm_r {parent=clavicle_r, loc=<-15.5111 0 0>, rot=<0.0226 -0.3234 -0.0088 -0.9459>}
 lowerarm_r {parent=upperarm_r, loc=<-26.6514 0 0>, rot=<0.0391 -0.0099 -0.2453 -0.9686>}
 hand_r {parent=lowerarm_r, loc=<-22.3491 0 0>, rot=<-0.7081 -0.0507 0.0975 -0.6975>}
 thumb_01_r {parent=hand_r, loc=<-1.3118 0.3191 2.459>, rot=<0.6546 -0.4042 0.0664 -0.6354>}
 thumb_02_r {parent=thumb_01_r, loc=<-5.9336 0.0448 -0.0401>, rot=<-0.0037 0.0213 0.298 -0.9543>}
 thumb_03_r {parent=thumb_02_r, loc=<-2.2223 -0.3682 0.0907>, rot=<0.0118 -0.013 0.0066 -0.9998>}
 index_metacarpal_r {parent=hand_r, loc=<-1.6535 -0.0183 2.96>, rot=<0.0158 -0.0124 -0.0173 -0.9996>}
 index_01_r {parent=index_metacarpal_r, loc=<-8.3008 -0.0357 0.0524>, rot=<-0.0044 0.0485 0.0924 -0.9945>}
 index_02_r {parent=index_01_r, loc=<-3.3587 0.3004 -0.0452>, rot=<-0.002 0.0027 0.3225 -0.9466>}
 index_03_r {parent=index_02_r, loc=<-2.5578 0.123 0.0252>, rot=<-0.002 0.0067 0.2558 -0.9667>}
 middle_metacarpal_r {parent=hand_r, loc=<-2.1856 0.1725 0.5984>, rot=<0.006 0.0168 -0.0329 -0.9993>}
 middle_01_r {parent=middle_metacarpal_r, loc=<-7.3275 0.2447 -0.0129>, rot=<-0.0208 0.0042 0.0858 -0.9961>}
 middle_02_r {parent=middle_01_r, loc=<-4.0123 -0.1401 -0.1175>, rot=<-0.0218 0.0258 0.3512 -0.9357>}
 middle_03_r {parent=middle_02_r, loc=<-2.6077 -0.0534 -0.1373>, rot=<0.0082 0.0188 0.1708 -0.9851>}
 ring_metacarpal_r {parent=hand_r, loc=<-2.2074 0.3625 -1.3638>, rot=<-0.0284 0.0722 -0.0458 -0.9959>}
 ring_01_r {parent=ring_metacarpal_r, loc=<-7.0295 0.6197 0.1446>, rot=<-0.0489 -0.0252 0.1892 -0.9804>}
 ring_02_r {parent=ring_01_r, loc=<-3.6174 -0.4647 0.1027>, rot=<0 0 0.2806 -0.9598>}
 ring_03_r {parent=ring_02_r, loc=<-2.4152 -0.114 -0.0199>, rot=<0 0 0.1189 -0.9929>}
 pinky_metacarpal_r {parent=hand_r, loc=<-1.7616 0.9753 -3.3814>, rot=<-0.0397 0.0814 -0.069 -0.9935>}
 pinky_01_r {parent=pinky_metacarpal_r, loc=<-7.1709 0.4831 0.1094>, rot=<-0.0654 -0.0035 0.2242 -0.9723>}
 pinky_02_r {parent=pinky_01_r, loc=<-2.8274 -0.0893 0.1043>, rot=<0.0037 0.0011 0.2854 -0.9584>}
 pinky_03_r {parent=pinky_02_r, loc=<-2.2046 -0.0307 0.0078>, rot=<-0.0002 0.0098 0.1445 -0.9895>}
 deform_wrist_dn_base_r {parent=hand_r, loc=<-0.0004 2.2345 -0.0001>, rot=<0 0 0 -1>}
 deform_wrist_dn_r {parent=deform_wrist_dn_base_r, loc=<0 0 0>, rot=<0 0 0 1>}
 deform_wrist_up_base_r {parent=hand_r, loc=<0.0003 -2.2684 0.0001>, rot=<0 0 0 -1>}
 deform_wrist_up_r {parent=deform_wrist_up_base_r, loc=<0 0 0>, rot=<0 0 0 1>}
 lowerarm_twist_02_r {parent=lowerarm_r, loc=<-9.6199 0 0>, rot=<0 0 0 -1>}
 lowerarm_twist_01_r {parent=lowerarm_r, loc=<-17.8549 0 0>, rot=<0 0 0 -1>}
 deform_elbow_in_base_r {parent=lowerarm_r, loc=<-0.2111 -5.2591 -0.8275>, rot=<0 0 0 -1>}
 deform_elbow_in_r {parent=deform_elbow_in_base_r, loc=<-0.0001 0 0.0002>, rot=<0 0 0 1>}
 deform_elbow_out_base_r {parent=lowerarm_r, loc=<0 4 -0.0001>, rot=<0 0 0 -1>}
 deform_elbow_out_r {parent=deform_elbow_out_base_r, loc=<0 0 0>, rot=<0 0 0 1>}
 upperarm_twist_01_r {parent=upperarm_r, loc=<-0.0041 0 0>, rot=<0 0 0 -1>}
 deform_upperarm_bk_base_r {parent=upperarm_twist_01_r, loc=<-1.0814 7.2192 1.9455>, rot=<0 0 0 1>}
 deform_upperarm_bk_r {parent=deform_upperarm_bk_base_r, loc=<0 0 0>, rot=<0 0 0 1>}
 deform_upperarm_fr_base_r {parent=upperarm_twist_01_r, loc=<0.3689 -7.0506 2.7652>, rot=<0 0 0 1>}
 deform_upperarm_fr_r {parent=deform_upperarm_fr_base_r, loc=<0 0 0>, rot=<0 0 0 1>}
 deform_upperarm_up_base_r {parent=upperarm_twist_01_r, loc=<3.2214 1.1391 -3.3308>, rot=<0 0 0 1>}
 deform_upperarm_up_r {parent=deform_upperarm_up_base_r, loc=<0 0 0>, rot=<0 0 0 -1>}
 upperarm_twist_02_r {parent=upperarm_r, loc=<-13.7941 0 0>, rot=<0 0 0 -1>}
 dyn_hood_MSTR_L {parent=spine_05, loc=<11.3138 0.1529 0>, rot=<0.3273 -0.5068 -0.0948 -0.7919>}
 dyn_hood_l_01 {parent=dyn_hood_MSTR_L, loc=<0 0 0>, rot=<0 0 0 1>}
 dyn_hood_l_02 {parent=dyn_hood_l_01, loc=<14.4792 0 0>, rot=<0 0 0 1>}
 dyn_hood_fr_MSTR_L {parent=spine_05, loc=<11.3138 0.1529 0>, rot=<0.1108 -0.5966 -0.2666 -0.7488>}
 dyn_hood_fr_l_01 {parent=dyn_hood_fr_MSTR_L, loc=<0 0 0>, rot=<0 0 0 1>}
 dyn_hood_fr_l_02 {parent=dyn_hood_fr_l_01, loc=<15.2885 0 0>, rot=<0 0 0 1>}
 dyn_hood_mid_MSTR_L {parent=spine_05, loc=<11.3138 0.1529 0>, rot=<0.4042 -0.382 0.1586 -0.8158>}
 dyn_hood_mid_l_01 {parent=dyn_hood_mid_MSTR_L, loc=<0 0 0>, rot=<0 0 0 1>}
 dyn_hood_mid_l_02 {parent=dyn_hood_mid_l_01, loc=<15.0962 0 0>, rot=<0 0 0 1>}
 dyn_hood_fr_MSTR_R {parent=spine_05, loc=<11.3141 0.1529 0>, rot=<0.7445 0.2872 0.587 -0.1368>}
 dyn_hood_fr_r_01 {parent=dyn_hood_fr_MSTR_R, loc=<0 0 0>, rot=<0 0 0 1>}
 dyn_hood_fr_r_02 {parent=dyn_hood_fr_r_01, loc=<15.2884 0 0>, rot=<0 0 0 1>}
 dyn_hood_MSTR_R {parent=spine_05, loc=<11.3141 0.1529 0>, rot=<0.7975 0.0859 0.5083 -0.3134>}
 dyn_hood_r_01 {parent=dyn_hood_MSTR_R, loc=<0 0 0>, rot=<0 0 0 1>}
 dyn_hood_r_02 {parent=dyn_hood_r_01, loc=<14.4792 0 0>, rot=<0 0 0 1>}
 dyn_hood_mid_MSTR_R {parent=spine_05, loc=<11.3141 0.1529 0>, rot=<0.7605 -0.1109 0.3985 -0.5007>}
 dyn_hood_mid_r_01 {parent=dyn_hood_mid_MSTR_R, loc=<0 0 0>, rot=<0 0 0 1>}
 dyn_hood_mid_r_02 {parent=dyn_hood_mid_r_01, loc=<15.0962 0 0>, rot=<0 0 0 1>}
 dyn_hood_c_MSTR {parent=spine_05, loc=<11.3138 0.1529 0>, rot=<0.6531 -0.2711 0.2711 -0.6531>}
 dyn_hood_c_01 {parent=dyn_hood_c_MSTR, loc=<0 0 0>, rot=<0 0 0 1>}
 dyn_hood_c_02 {parent=dyn_hood_c_01, loc=<15.8266 0 0>, rot=<0 0 0 1>}
 dyn_hood_bk_01 {parent=dyn_hood_c_02, loc=<0 0 0>, rot=<0 -0.9125 0 -0.4091>}
 dyn_hood_bk_02 {parent=dyn_hood_bk_01, loc=<6.6017 0 0>, rot=<0 0.0605 0 -0.9982>}
 dyn_hood_bk_03 {parent=dyn_hood_bk_02, loc=<6.2015 0 0>, rot=<0.1094 -0.0196 -0.0084 -0.9938>}
 dyn_hood_bk_04 {parent=dyn_hood_bk_03, loc=<6.2909 0 0>, rot=<0 -0.0175 0 -0.9998>}
 deform_pec_aim_base_l {parent=spine_04, loc=<7.8728 12.6388 -2.4123>, rot=<0.0102 -0.4602 0.3979 -0.7936>}
 deform_pec_aim_l {parent=deform_pec_aim_base_l, loc=<0 0 0>, rot=<0 0 0 1>}
 deform_pec_base_l {parent=deform_pec_aim_l, loc=<10.766 0 0>, rot=<0 0 0 1>}
 deform_pec_l {parent=deform_pec_base_l, loc=<0.055 0 0>, rot=<0 0 0 1>}
 deform_pec_aim_base_r {parent=spine_04, loc=<7.8723 12.6388 2.4123>, rot=<-0.4602 -0.0102 0.7936 0.3979>}
 deform_pec_aim_r {parent=deform_pec_aim_base_r, loc=<0 0 0>, rot=<0 0 0 1>}
 deform_pec_base_r {parent=deform_pec_aim_r, loc=<-10.7662 -0.0006 -0.0007>, rot=<0 0 0 1>}
 deform_pec_r {parent=deform_pec_base_r, loc=<0 0 0>, rot=<0 0 0 1>}
 dyn_jacket_bk_MSTR {parent=spine_04, loc=<2.2851 -9.8336 0.0295>, rot=<-0.0335 -0.7125 0.7 0.0352>}
 dyn_jacket_bk_01 {parent=dyn_jacket_bk_MSTR, loc=<1 0 0>, rot=<0 0 0 1>}
 dyn_jacket_bk_02 {parent=dyn_jacket_bk_01, loc=<11.8396 0 0>, rot=<0 0.1682 0 -0.9858>}
 dyn_jacket_bk_03 {parent=dyn_jacket_bk_02, loc=<14.6301 0 0>, rot=<0 0 0 1>}
 dyn_jacket_bk_l_01 {parent=dyn_jacket_bk_03, loc=<9.0878 7.4515 -0.3269>, rot=<0.1632 0.0275 -0.0781 -0.9831>}
 dyn_jacket_bk_l_02 {parent=dyn_jacket_bk_l_01, loc=<10.1225 0 0>, rot=<0.0118 -0.0056 0.0207 -0.9997>}
 dyn_jacket_bk_r_01 {parent=dyn_jacket_bk_03, loc=<9.0472 -7.4847 -0.5897>, rot=<-0.116 0.0212 0.0828 -0.9896>}
 dyn_jacket_bk_r_02 {parent=dyn_jacket_bk_r_01, loc=<10.1225 0 0>, rot=<0 0 0 1>}
 dyn_jacket_MSTR_L {parent=spine_04, loc=<0.7442 1.9312 -14.9923>, rot=<-0.0294 -0.9964 0.0658 0.0457>}
 dyn_jacket_l_01 {parent=dyn_jacket_MSTR_L, loc=<1 0 0>, rot=<0 0 0 1>}
 dyn_jacket_l_02 {parent=dyn_jacket_l_01, loc=<11.6008 0 0>, rot=<0.073 0.1612 0.063 -0.9822>}
 dyn_jacket_l_03 {parent=dyn_jacket_l_02, loc=<13.7336 0 0>, rot=<-0.0584 0.0115 0.0143 -0.9981>}
 dyn_jacket_l_04 {parent=dyn_jacket_l_03, loc=<9.8341 0 0>, rot=<-0.0096 -0.0016 0.0024 -0.9999>}
 dyn_jacket_l_05 {parent=dyn_jacket_l_04, loc=<11.0587 0 0>, rot=<-0.0005 -0.0268 0.0001 -0.9996>}
 dyn_jacket_fr_MSTR_L {parent=spine_04, loc=<4.3007 12.4186 -10.2415>, rot=<0.0534 -0.908 -0.4155 -0.0102>}
 dyn_jacket_fr_l_01 {parent=dyn_jacket_fr_MSTR_L, loc=<1 0 0>, rot=<0 0 0 1>}
 dyn_jacket_fr_l_02 {parent=dyn_jacket_fr_l_01, loc=<8.6517 0 0>, rot=<0 0 0 1>}
 dyn_jacket_fr_l_03 {parent=dyn_jacket_fr_l_02, loc=<8.3704 0 0>, rot=<0.0208 0.0622 0.0297 -0.9974>}
 dyn_jacket_fr_l_04 {parent=dyn_jacket_fr_l_03, loc=<13.8554 0 0>, rot=<-0.0366 0.0141 0.015 -0.9991>}
 dyn_jacket_fr_l_05 {parent=dyn_jacket_fr_l_04, loc=<9.6925 0 0>, rot=<-0.006 -0.0063 -0.0073 -0.9999>}
 dyn_jacket_fr_l_06 {parent=dyn_jacket_fr_l_05, loc=<11.887 0 0>, rot=<-0.0349 -0.0001 0.0002 -0.9994>}
 dyn_jacket_mid_MSTR_L {parent=spine_04, loc=<1.4035 -7.2403 -12.0799>, rot=<-0.0611 -0.9362 0.3382 0.0737>}
 dyn_jacket_mid_l_01 {parent=dyn_jacket_mid_MSTR_L, loc=<1 0 0>, rot=<0 0 0 1>}
 dyn_jacket_mid_l_02 {parent=dyn_jacket_mid_l_01, loc=<11.5499 0 0>, rot=<0.0216 0.2127 0.012 -0.9768>}
 dyn_jacket_mid_l_03 {parent=dyn_jacket_mid_l_02, loc=<13.8371 0 0>, rot=<0.0003 0.0144 0.0238 -0.9996>}
 dyn_jacket_mid_l_04 {parent=dyn_jacket_mid_l_03, loc=<9.63 0 0>, rot=<0 -0.0026 0.017 -0.9999>}
 dyn_jacket_mid_l_05 {parent=dyn_jacket_mid_l_04, loc=<10.5418 0 0>, rot=<0 -0.0166 0.0001 -0.9999>}
 dyn_jacket_fr_MSTR_R {parent=spine_04, loc=<4.3002 12.4185 10.2414>, rot=<0.0084 0.447 0.8929 -0.0537>}
 dyn_jacket_fr_r_01 {parent=dyn_jacket_fr_MSTR_R, loc=<1 0 0>, rot=<0 0 0 1>}
 dyn_jacket_fr_r_02 {parent=dyn_jacket_fr_r_01, loc=<8.652 0 0>, rot=<0 0 0 1>}
 dyn_jacket_fr_r_03 {parent=dyn_jacket_fr_r_02, loc=<8.3699 0 0>, rot=<-0.021 0.06 -0.034 -0.9974>}
 dyn_jacket_fr_r_04 {parent=dyn_jacket_fr_r_03, loc=<13.8551 0 0>, rot=<0.0366 0.013 -0.0159 -0.9991>}
 dyn_jacket_fr_r_05 {parent=dyn_jacket_fr_r_04, loc=<9.6925 0 0>, rot=<0.006 -0.0058 0.0077 -0.9999>}
 dyn_jacket_fr_r_06 {parent=dyn_jacket_fr_r_05, loc=<11.8871 0 0>, rot=<-0.0187 0 0 -0.9998>}
 dyn_jacket_MSTR_R {parent=spine_04, loc=<0.7439 1.9311 14.9923>, rot=<-0.0464 -0.0928 0.9942 0.0282>}
 dyn_jacket_r_01 {parent=dyn_jacket_MSTR_R, loc=<1 0 0>, rot=<0 0 0 1>}
 dyn_jacket_r_02 {parent=dyn_jacket_r_01, loc=<11.6007 0 0>, rot=<-0.0715 0.1644 -0.0539 -0.9823>}
 dyn_jacket_r_03 {parent=dyn_jacket_r_02, loc=<13.7334 0 0>, rot=<-0.0588 0.0106 -0.015 -0.9981>}
 dyn_jacket_r_04 {parent=dyn_jacket_r_03, loc=<9.834 0 0>, rot=<0.0096 -0.0021 -0.0021 -0.9999>}
 dyn_jacket_r_05 {parent=dyn_jacket_r_04, loc=<11.0588 0 0>, rot=<0.0005 0 0 -1>}
 dyn_jacket_mid_MSTR_R {parent=spine_04, loc=<1.4038 -7.2403 12.0799>, rot=<-0.0747 -0.3542 0.9303 0.0598>}
 dyn_jacket_mid_r_01 {parent=dyn_jacket_mid_MSTR_R, loc=<1 0 0>, rot=<0 0 0 1>}
 dyn_jacket_mid_r_02 {parent=dyn_jacket_mid_r_01, loc=<11.5503 0 0>, rot=<0 0.213 0 -0.977>}
 dyn_jacket_mid_r_03 {parent=dyn_jacket_mid_r_02, loc=<13.8369 0 0>, rot=<0.0284 0.0169 -0.0221 -0.9992>}
 dyn_jacket_mid_r_04 {parent=dyn_jacket_mid_r_03, loc=<9.6301 0 0>, rot=<-0.0273 -0.0007 -0.0172 -0.9995>}
 dyn_jacket_mid_r_05 {parent=dyn_jacket_mid_r_04, loc=<10.5418 0 0>, rot=<-0.0175 0 0 -0.9998>}
 dyn_necklace_MSTR {parent=spine_04, loc=<16.3078 8.6239 0>, rot=<-0.2033 -0.6784 -0.6758 -0.2044>}
 dyn_necklace_01 {parent=dyn_necklace_MSTR, loc=<1 0 0>, rot=<0 0 0 1>}
 dyn_necklace_02 {parent=dyn_necklace_01, loc=<4.8894 0 0>, rot=<0.0002 -0.0792 -0.0022 -0.9969>}
 dyn_necklace_03 {parent=dyn_necklace_02, loc=<7.0171 0 0>, rot=<-0.0007 -0.2795 0.0023 -0.9601>}
 thigh_l {parent=pelvis, loc=<-2.5693 -0.0109 -7.7501>, rot=<0.0012 0.0572 0.0155 -0.9982>}
 calf_l {parent=thigh_l, loc=<-39.0484 0 0>, rot=<0.0006 -0.0161 -0.0391 -0.9991>}
 foot_l {parent=calf_l, loc=<-42.561 0.1077 0.4292>, rot=<0.0418 -0.0455 0.0343 -0.9975>}
 ball_l {parent=foot_l, loc=<-6.1599 13.8264 0.6806>, rot=<0.0001 -0.0001 -0.7063 -0.7079>}
 calf_twist_02_l {parent=calf_l, loc=<-18.8064 -0.0967 0.0397>, rot=<0.0237 0.0095 -0.0035 -0.9997>}
 calf_twist_01_l {parent=calf_l, loc=<-35.813 -0.0967 0.0397>, rot=<0.0237 0.0095 -0.0035 -0.9997>}
 deform_knee_in_base_l {parent=calf_l, loc=<1.298 -5.2957 -0.0496>, rot=<0 0 0 -1>}
 deform_knee_in_l {parent=deform_knee_in_base_l, loc=<0 0 0>, rot=<0 0 0 1>}
 deform_knee_out_base_l {parent=calf_l, loc=<0 4 0>, rot=<0 0 0 -1>}
 deform_knee_out_l {parent=deform_knee_out_base_l, loc=<0 0 0>, rot=<0 0 0 1>}
 thigh_twist_01_l {parent=thigh_l, loc=<-21.6303 0 0>, rot=<0.0013 0 0 -1>}
 deform_groin_base_l {parent=thigh_l, loc=<-2.1908 8.1142 -1.7442>, rot=<0 0 0 -1>}
 deform_groin_l {parent=deform_groin_base_l, loc=<0 0 0>, rot=<0 0 0 1>}
 thigh_r {parent=pelvis, loc=<-2.5693 -0.0109 7.7501>, rot=<0.0572 -0.0012 0.9982 0.0155>}
 calf_r {parent=thigh_r, loc=<39.0484 0 0>, rot=<0.0006 -0.0161 -0.0391 -0.9991>}
 foot_r {parent=calf_r, loc=<42.561 -0.1077 -0.4292>, rot=<0.0418 -0.0455 0.0343 -0.9975>}
 ball_r {parent=foot_r, loc=<6.1599 -13.8264 -0.6806>, rot=<0.0001 -0.0001 -0.7063 -0.7079>}
 calf_twist_02_r {parent=calf_r, loc=<18.8064 0.0967 -0.0397>, rot=<0.0237 0.0095 -0.0035 -0.9997>}
 calf_twist_01_r {parent=calf_r, loc=<35.813 0.0967 -0.0397>, rot=<0.0237 0.0095 -0.0035 -0.9997>}
 deform_knee_in_base_r {parent=calf_r, loc=<-1.298 5.2957 0.0496>, rot=<0 0 0 -1>}
 deform_knee_in_r {parent=deform_knee_in_base_r, loc=<0 0 0>, rot=<0 0 0 1>}
 deform_knee_out_base_r {parent=calf_r, loc=<0 -4 0>, rot=<0 0 0 -1>}
 deform_knee_out_r {parent=deform_knee_out_base_r, loc=<0 0 0>, rot=<0 0 0 1>}
 thigh_twist_01_r {parent=thigh_r, loc=<21.6303 0 0>, rot=<0.0013 0 0 -1>}
 deform_groin_base_r {parent=thigh_r, loc=<2.1908 -8.1142 1.7442>, rot=<0 0 0 -1>}
 deform_groin_r {parent=deform_groin_base_r, loc=<0 0 0>, rot=<0 0 0 1>}
 deform_glute_base_l {parent=pelvis, loc=<-8.8064 -8.8222 -8.4287>, rot=<0.0275 0.0501 -0.4472 -0.8926>}
 deform_glute_l {parent=deform_glute_base_l, loc=<0 0 0>, rot=<0 0 0 1>}
 deform_glute_base_r {parent=pelvis, loc=<-8.8062 -8.8222 8.4287>, rot=<0.0501 -0.0275 0.8926 -0.4472>}
 deform_glute_r {parent=deform_glute_base_r, loc=<0 0 0>, rot=<0 0 0 1>}
 dyn_buckle_MSTR {parent=pelvis, loc=<5.9471 11.1055 0.0003>, rot=<-0.5089 -0.491 -0.491 -0.5089>}
 dyn_buckle {parent=dyn_buckle_MSTR, loc=<0 0 0>, rot=<0 0 0 1>}
 dyn_chain_01 {parent=dyn_buckle, loc=<0.1334 4.7661 -3.0145>, rot=<0.3797 -0.2798 -0.7099 -0.5231>}
 dyn_chain_02 {parent=dyn_chain_01, loc=<6.5975 0 0>, rot=<0.0441 0.0663 -0.0339 -0.9963>}
 dyn_chain_03 {parent=dyn_chain_02, loc=<5.6033 0 0>, rot=<0.0892 0.1958 -0.1173 -0.9695>}
 dyn_chain_04 {parent=dyn_chain_03, loc=<6.8676 0 0>, rot=<0.0833 0.2103 -0.3696 -0.9013>}
 dyn_chain_05 {parent=dyn_chain_04, loc=<6.533 0 0>, rot=<-0.0556 0.1875 -0.2691 -0.9431>}
 dyn_chain_06 {parent=dyn_chain_05, loc=<5.3551 0 0>, rot=<-0.0928 0.1764 -0.1417 -0.9696>}
 dyn_chain_07 {parent=dyn_chain_06, loc=<5.2427 0 0>, rot=<-0.1194 0.0713 -0.1086 -0.9843>}
 dyn_chain_08 {parent=dyn_chain_07, loc=<5.6039 0 0>, rot=<-0.2089 0.034 -0.1526 -0.9654>}
 dyn_chain_09 {parent=dyn_chain_08, loc=<6.0424 0 0>, rot=<0 0 0 1>}
 dyn_candy_cane {parent=pelvis, loc=<8.1353 6.685 12.2763>, rot=<0.2708 -0.3883 0.8744 -0.1064>}
 dyn_belt {parent=pelvis, loc=<6.7587 7.1649 13.2302>, rot=<0.2708 -0.3883 0.8744 -0.1064>}
]
variables [
 isCloth {type=bool, default=False}
 isCrouchMoving {type=bool, default=False}
 isCrouching {type=bool, default=False}
 isFrontEnd {type=bool, default=False}
 isRailGrinding {type=bool, default=False}
 isWearingCape {type=bool, default=False}
 isWearingPack {type=bool, default=False}
]
