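# Physics asset for the AlmondSplash male medium player body that sets up a long jacket to swing against the character.
# Capsule chains for front, side, mid and back panels hang from tiny MSTR anchor bodies with limited swing and twist, the back splitting into two tails.
# Angularly free cross-constraints tie neighboring chains into a connected skirt.
# Spine, pelvis and leg capsules provide body shapes, and a huge root sphere below likely approximates the ground.
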PhysicsAsset /BRCosmetics/Characters/Player/Male/Medium/Bodies/M_MED_AlmondSplash/Meshes/M_MED_AlmondSplash_Physics
bodies {
 calf_l [{shapeType=Capsule, length=31.7339, radius=8.4601, shapeName=calf_l_capsule, center=<-18.2339 -2.3109 0>, rotation=<rot 90 0 0>}]
 calf_r [{shapeType=Capsule, length=32.702, radius=8.5863, shapeName=calf_r_capsule, center=<18.2339 2.1876 0>, rotation=<rot -90 180 180>}]
 dyn_jacket_MSTR_L [{shapeType=Capsule, length=0.5, radius=0.25, shapeName=dyn_jacket_MSTR_L_capsule, rotation=<rot 90 0 0>}]
 dyn_jacket_MSTR_R [{shapeType=Capsule, length=0.5, radius=0.25, shapeName=dyn_jacket_MSTR_R_capsule, rotation=<rot 90 0 0>}]
 dyn_jacket_bk_01 [{shapeType=Capsule, length=10.0312, radius=4.4688, shapeName=dyn_coat_bk_01_capsule, center=<13 0 5>, rotation=<rot -76.9676 -0.0001 0>}]
 dyn_jacket_bk_02 [{shapeType=Capsule, length=17.1562, radius=6.8438, shapeName=dyn_coat_fr_03_L_capsule, center=<13 0 5>, rotation=<rot -82.8531 -179.9978 180>}]
 dyn_jacket_bk_03 [{shapeType=Capsule, length=14, radius=6.5, shapeName=dyn_coat_bk_03_capsule, center=<13 0 5>, rotation=<rot -90 -179.9605 179.9604>}]
 dyn_jacket_bk_MSTR [{shapeType=Capsule, length=0.5, radius=0.25, shapeName=dyn_jacket_bk_MSTR_capsule, rotation=<rot 90 0 0>}]
 dyn_jacket_bk_l_01 [{shapeType=Capsule, length=14, radius=6.5, shapeName=dyn_coat_fr_05_R_capsule, center=<7.4809 0 3.1655>, rotation=<rot 0 0 90>}]
 dyn_jacket_bk_l_02 [{shapeType=Capsule, length=23.9062, radius=8.5, shapeName=dyn_coat_fr_06_R_capsule, center=<8.8377 0.0002 5>, rotation=<rot -90 90 179.9972>}]
 dyn_jacket_bk_r_01 [{shapeType=Capsule, length=14, radius=6.5, shapeName=dyn_coat_fr_05_R_capsule, center=<7.4809 0 3.1655>, rotation=<rot 0 0 90>}]
 dyn_jacket_bk_r_02 [{shapeType=Capsule, length=23.9062, radius=8.5, shapeName=dyn_coat_fr_06_R_capsule, center=<8.8377 0.0002 5>, rotation=<rot -90 90 179.9972>}]
 dyn_jacket_fr_MSTR_L [{shapeType=Capsule, length=0.5, radius=0.25, shapeName=dyn_jacket_fr_MSTR_L_capsule, rotation=<rot 90 0 0>}]
 dyn_jacket_fr_MSTR_R [{shapeType=Capsule, length=0.5, radius=0.25, shapeName=dyn_jacket_fr_MSTR_R_capsule, rotation=<rot 90 0 0>}]
 dyn_jacket_fr_l_01 [{
   shapeType=Capsule
   length=13.6875
   radius=3.5
   shapeName=dyn_coat_fr_01_L_capsule
   center=<13.0001 -2.8519 3>
   rotation=<rot -83.8458 0.0022 0>
  }]
 dyn_jacket_fr_l_02 [{
   shapeType=Capsule
   length=15.9375
   radius=4.5625
   shapeName=dyn_coat_fr_01_L_capsule
   center=<13.0001 -2.8519 3>
   rotation=<rot -82.5295 0.0023 0>
  }]
 dyn_jacket_fr_l_03 [{
   shapeType=Capsule
   length=18.5
   radius=5.5
   shapeName=dyn_coat_fr_03_L_capsule
   center=<13.0001 -2.8519 3>
   rotation=<rot -90 179.9965 -179.9943>
  }]
 dyn_jacket_fr_l_04 [{
   shapeType=Capsule
   length=13.5
   radius=6
   shapeName=dyn_coat_fr_04_L_capsule
   center=<13.0001 -2.8519 3>
   rotation=<rot -90 160.5284 -160.5262>
  }]
 dyn_jacket_fr_l_05 [{
   shapeType=Capsule
   length=14
   radius=6.5
   shapeName=dyn_coat_fr_05_L_capsule
   center=<13.0001 -2.8519 3>
   rotation=<rot -90 179.3671 -179.3648>
  }]
 dyn_jacket_fr_l_06 [{
   shapeType=Capsule
   length=15.5
   radius=8.5
   shapeName=dyn_coat_fr_06_L_capsule
   center=<11.2569 -2.852 5>
   rotation=<rot -90 179.9738 -179.9729>
  }]
 dyn_jacket_fr_r_01 [{
   shapeType=Capsule
   length=13.3438
   radius=3.8438
   shapeName=dyn_coat_fr_01_R_capsule
   center=<13.0001 2.0036 3>
   rotation=<rot -81.3148 -0.0023 0>
  }]
 dyn_jacket_fr_r_02 [{
   shapeType=Capsule
   length=15.9062
   radius=4.5938
   shapeName=dyn_coat_fr_01_R_capsule
   center=<13.0164 2.0036 3>
   rotation=<rot -82.5295 -0.0023 0>
  }]
 dyn_jacket_fr_r_03 [{
   shapeType=Capsule
   length=17.5683
   radius=6.4317
   shapeName=dyn_coat_fr_03_R_capsule
   center=<13.0001 2.0036 3>
   rotation=<rot -90 -179.9991 179.997>
  }]
 dyn_jacket_fr_r_04 [{shapeType=Capsule, length=13.5, radius=6, shapeName=dyn_coat_fr_04_L_capsule, center=<12.9999 2.0036 3>, rotation=<rot -90 90 -89.9978>}]
 dyn_jacket_fr_r_05 [{shapeType=Capsule, length=14, radius=6.5, shapeName=dyn_coat_fr_05_R_capsule, center=<13.0001 2.0036 3>, rotation=<rot -90 0 -0.0022>}]
 dyn_jacket_fr_r_06 [{shapeType=Capsule, length=15.5, radius=8.5, shapeName=dyn_coat_fr_06_R_capsule, center=<11.257 2.0037 5>, rotation=<rot -90 0 -0.0031>}]
 dyn_jacket_l_01 [{shapeType=Capsule, length=10.4303, radius=5.3822, shapeName=dyn_coat_mid_01_L_capsule, center=<13 0 5>, rotation=<rot -82.9332 0.002 0>}]
 dyn_jacket_l_02 [{shapeType=Capsule, length=18.5, radius=5.5, shapeName=dyn_coat_fr_03_L_capsule, center=<13 0 5>, rotation=<rot -90 179.9965 -179.9943>}]
 dyn_jacket_l_03 [{shapeType=Capsule, length=14, radius=6.5, shapeName=dyn_coat_mid_03_L_capsule, center=<13 0 5>, rotation=<rot -90 179.9956 -179.9938>}]
 dyn_jacket_l_04 [{
   shapeType=Capsule
   length=14
   radius=6.5
   shapeName=dyn_coat_fr_05_R_capsule
   center=<13 0 2.9223>
   rotation=<rot -83.547 179.9977 -179.9998>
  }]
 dyn_jacket_l_05 [{
   shapeType=Capsule
   length=21.6889
   radius=8.5
   shapeName=dyn_coat_fr_06_R_capsule
   center=<13.0003 5.1592 3.6638>
   rotation=<rot -72.1716 -90.0027 0.0001>
  }]
 dyn_jacket_mid_MSTR_L [{shapeType=Capsule, length=0.5, radius=0.25, shapeName=dyn_jacket_mid_MSTR_L_capsule, rotation=<rot 90 0 0>}]
 dyn_jacket_mid_MSTR_R [{shapeType=Capsule, length=0.5, radius=0.25, shapeName=dyn_jacket_mid_MSTR_R_capsule, rotation=<rot 90 0 0>}]
 dyn_jacket_mid_l_01 [{shapeType=Capsule, length=8.8679, radius=5.6321, shapeName=dyn_coat_mid_01_L_capsule, center=<13 0 5>, rotation=<rot -76.9675 0.0019 0.0001>}]
 dyn_jacket_mid_l_02 [{
   shapeType=Capsule
   length=17.1562
   radius=6.8438
   shapeName=dyn_coat_fr_03_L_capsule
   center=<12.8989 0 4.4766>
   rotation=<rot -89.6287 0.0101 -0.0078>
  }]
 dyn_jacket_mid_l_03 [{shapeType=Capsule, length=14, radius=6.5, shapeName=dyn_coat_mid_03_L_capsule, center=<13 0 5>, rotation=<rot -90 179.9956 -179.9938>}]
 dyn_jacket_mid_l_04 [{shapeType=Capsule, length=14, radius=6.5, shapeName=dyn_coat_fr_05_R_capsule, center=<13 0 5>, rotation=<rot 0 0 90>}]
 dyn_jacket_mid_l_05 [{shapeType=Capsule, length=26.8125, radius=8.5, shapeName=dyn_coat_fr_06_R_capsule, center=<13 0 5>, rotation=<rot -90 -17.4099 -72.5929>}]
 dyn_jacket_mid_r_01 [{shapeType=Capsule, length=10.0312, radius=4.4688, shapeName=dyn_coat_mid_01_R_capsule, center=<13 0 5>, rotation=<rot -76.9674 -0.0019 0>}]
 dyn_jacket_mid_r_02 [{
   shapeType=Capsule
   length=17.1562
   radius=6.8438
   shapeName=dyn_coat_fr_03_L_capsule
   center=<12.8989 0 4.4766>
   rotation=<rot -79.0643 -179.9977 179.9998>
  }]
 dyn_jacket_mid_r_03 [{shapeType=Capsule, length=14, radius=6.5, shapeName=dyn_coat_mid_03_R_capsule, center=<13 0 5>, rotation=<rot -90 -0.021 0.0191>}]
 dyn_jacket_mid_r_04 [{shapeType=Capsule, length=14, radius=6.5, shapeName=dyn_coat_fr_05_R_capsule, center=<13 0 5>, rotation=<rot 0 0 90>}]
 dyn_jacket_mid_r_05 [{shapeType=Capsule, length=26.8125, radius=8.5, shapeName=dyn_coat_fr_06_R_capsule, center=<13 0 5>, rotation=<rot -90 -17.4099 -72.5929>}]
 dyn_jacket_r_01 [{
   shapeType=Capsule
   length=11.25
   radius=4.5625
   shapeName=dyn_coat_mid_01_R_capsule
   center=<13.0944 0 4.2387>
   rotation=<rot -82.9332 -0.002 0>
  }]
 dyn_jacket_r_02 [{shapeType=Capsule, length=18.5, radius=5.5, shapeName=dyn_coat_fr_03_L_capsule, center=<13 0 5>, rotation=<rot -90 179.9965 -179.9943>}]
 dyn_jacket_r_03 [{shapeType=Capsule, length=16.4812, radius=7.66, shapeName=dyn_coat_mid_03_R_capsule, center=<13 0 5>, rotation=<rot -90 -0.021 0.0191>}]
 dyn_jacket_r_04 [{
   shapeType=Capsule
   length=14
   radius=6.5
   shapeName=dyn_coat_fr_05_R_capsule
   center=<13.1621 0 4.1678>
   rotation=<rot -82.5833 179.9979 179.9998>
  }]
 dyn_jacket_r_05 [{
   shapeType=Capsule
   length=24.6459
   radius=8.5
   shapeName=dyn_coat_fr_06_R_capsule
   center=<12.9999 -2.8277 2.6849>
   rotation=<rot -74.7519 89.9971 -180>
  }]
 pelvis [{shapeType=Capsule, length=11.4853, radius=10.6755, shapeName=pelvis_capsule}]
 root [{shapeType=Sphere, radius=300, shapeName=root_sphere, center=<0 0 -289.119>}]
 spine_01 [{shapeType=Capsule, length=13.9141, radius=9.0899, shapeName=spine_01_capsule, center=<3.5975 2.3284 0>}]
 spine_02 [{shapeType=Capsule, length=4.2847, radius=11.5003, shapeName=spine_02_capsule, center=<4.8113 2.6414 0>}]
 spine_03 [{shapeType=Capsule, length=11.4441, radius=9.2565, shapeName=spine_03_capsule, center=<4.4602 2.6414 0>}]
 spine_04 [{shapeType=Capsule, length=7.7161, radius=11.5939, shapeName=spine_04_capsule}]
 spine_05 [{shapeType=Capsule, length=7.3931, radius=13.0299, shapeName=spine_05_capsule, center=<-4.2896 -0.8314 0>, rotation=<rot 5.4975 -90 -0.4638>}]
 thigh_l [{shapeType=Capsule, length=32.3064, radius=9.1055, shapeName=thigh_l_capsule, center=<-18.2339 0 0>, rotation=<rot 90 0 0>}]
 thigh_r [{shapeType=Capsule, length=32.9843, radius=9.026, shapeName=thigh_r_capsule, center=<18.2339 -0.5596 0>, rotation=<rot -90 -90.3524 92.3971>}]
}
constraints [
 {bone1Name=dyn_jacket_l_01, bone2Name=dyn_jacket_MSTR_L, swing1LimitDegrees=20, swing2LimitDegrees=20, twistLimitDegrees=8}
 {bone1Name=dyn_jacket_l_02, bone2Name=dyn_jacket_l_01, swing1LimitDegrees=20, swing2LimitDegrees=20, twistLimitDegrees=4}
 {bone1Name=dyn_jacket_l_03, bone2Name=dyn_jacket_l_02, swing1LimitDegrees=8, swing2LimitDegrees=22, twistLimitDegrees=4}
 {bone1Name=dyn_jacket_l_04, bone2Name=dyn_jacket_l_03, swing1LimitDegrees=12, swing2LimitDegrees=32, twistLimitDegrees=4}
 {bone1Name=dyn_jacket_l_05, bone2Name=dyn_jacket_l_04, swing1LimitDegrees=12, swing2LimitDegrees=32, twistLimitDegrees=8}
 {bone1Name=dyn_jacket_r_01, bone2Name=dyn_jacket_MSTR_R, swing1LimitDegrees=20, swing2LimitDegrees=20, twistLimitDegrees=8}
 {bone1Name=dyn_jacket_r_02, bone2Name=dyn_jacket_r_01, swing1LimitDegrees=20, swing2LimitDegrees=20, twistLimitDegrees=4}
 {bone1Name=dyn_jacket_r_03, bone2Name=dyn_jacket_r_02, swing1LimitDegrees=8, swing2LimitDegrees=22, twistLimitDegrees=4}
 {bone1Name=dyn_jacket_r_04, bone2Name=dyn_jacket_r_03, swing1LimitDegrees=12, swing2LimitDegrees=32, twistLimitDegrees=4}
 {bone1Name=dyn_jacket_r_05, bone2Name=dyn_jacket_r_04, swing1LimitDegrees=12, swing2LimitDegrees=32, twistLimitDegrees=8}
 {bone1Name=dyn_jacket_bk_01, bone2Name=dyn_jacket_bk_MSTR, swing1LimitDegrees=20, swing2LimitDegrees=20, twistLimitDegrees=8}
 {bone1Name=dyn_jacket_bk_02, bone2Name=dyn_jacket_bk_01, swing1LimitDegrees=8, swing2LimitDegrees=24, twistLimitDegrees=4}
 {bone1Name=dyn_jacket_bk_03, bone2Name=dyn_jacket_bk_02, swing1LimitDegrees=8, swing2LimitDegrees=18, twistLimitDegrees=4}
 {bone1Name=dyn_jacket_bk_l_01, bone2Name=dyn_jacket_bk_03, swing1LimitDegrees=12, swing2LimitDegrees=20, twistLimitDegrees=8}
 {bone1Name=dyn_jacket_bk_l_02, bone2Name=dyn_jacket_bk_l_01, swing1LimitDegrees=12, swing2LimitDegrees=20, twistLimitDegrees=8}
 {bone1Name=dyn_jacket_bk_r_01, bone2Name=dyn_jacket_bk_03, swing1LimitDegrees=12, swing2LimitDegrees=20, twistLimitDegrees=8}
 {bone1Name=dyn_jacket_bk_r_02, bone2Name=dyn_jacket_bk_r_01, swing1LimitDegrees=12, swing2LimitDegrees=20, twistLimitDegrees=8}
 {bone1Name=dyn_jacket_mid_l_01, bone2Name=dyn_jacket_mid_MSTR_L, swing1LimitDegrees=10, swing2LimitDegrees=10, twistLimitDegrees=8}
 {bone1Name=dyn_jacket_mid_l_02, bone2Name=dyn_jacket_mid_l_01, swing1LimitDegrees=8, swing2LimitDegrees=24, twistLimitDegrees=4}
 {bone1Name=dyn_jacket_mid_l_03, bone2Name=dyn_jacket_mid_l_02, swing1LimitDegrees=8, swing2LimitDegrees=18, twistLimitDegrees=4}
 {bone1Name=dyn_jacket_mid_l_04, bone2Name=dyn_jacket_mid_l_03, swing1LimitDegrees=12, swing2LimitDegrees=32, twistLimitDegrees=4}
 {bone1Name=dyn_jacket_mid_l_05, bone2Name=dyn_jacket_mid_l_04, swing1LimitDegrees=12, swing2LimitDegrees=32, twistLimitDegrees=8}
 {bone1Name=dyn_jacket_mid_r_01, bone2Name=dyn_jacket_mid_MSTR_R, swing1LimitDegrees=10, swing2LimitDegrees=10, twistLimitDegrees=8}
 {bone1Name=dyn_jacket_mid_r_02, bone2Name=dyn_jacket_mid_r_01, swing1LimitDegrees=8, swing2LimitDegrees=24, twistLimitDegrees=4}
 {bone1Name=dyn_jacket_mid_r_03, bone2Name=dyn_jacket_mid_r_02, swing1LimitDegrees=8, swing2LimitDegrees=18, twistLimitDegrees=4}
 {bone1Name=dyn_jacket_mid_r_04, bone2Name=dyn_jacket_mid_r_03, swing1LimitDegrees=12, swing2LimitDegrees=32, twistLimitDegrees=4}
 {bone1Name=dyn_jacket_mid_r_05, bone2Name=dyn_jacket_mid_r_04, swing1LimitDegrees=12, swing2LimitDegrees=32, twistLimitDegrees=8}
 {bone1Name=dyn_jacket_fr_l_01, bone2Name=dyn_jacket_fr_MSTR_L, swing1LimitDegrees=10, swing2LimitDegrees=10, twistLimitDegrees=8}
 {bone1Name=dyn_jacket_fr_l_02, bone2Name=dyn_jacket_fr_l_01, swing1LimitDegrees=12, swing2LimitDegrees=12, twistLimitDegrees=4}
 {bone1Name=dyn_jacket_fr_l_03, bone2Name=dyn_jacket_fr_l_02, swing1LimitDegrees=20, swing2LimitDegrees=20, twistLimitDegrees=4}
 {bone1Name=dyn_jacket_fr_l_04, bone2Name=dyn_jacket_fr_l_03, swing1LimitDegrees=12, swing2LimitDegrees=32, twistLimitDegrees=4}
 {
  bone1Name=dyn_jacket_fr_l_05
  bone2Name=dyn_jacket_fr_l_04
  swing1LimitDegrees=12
  swing2LimitDegrees=32
  twistLimitDegrees=8
  twistMotion=Locked
 }
 {bone1Name=dyn_jacket_fr_l_06, bone2Name=dyn_jacket_fr_l_05, swing1LimitDegrees=12, swing2LimitDegrees=32, twistLimitDegrees=8}
 {bone1Name=dyn_jacket_fr_r_01, bone2Name=dyn_jacket_fr_MSTR_R, swing1LimitDegrees=10, swing2LimitDegrees=10, twistLimitDegrees=8}
 {bone1Name=dyn_jacket_fr_r_02, bone2Name=dyn_jacket_fr_r_01, swing1LimitDegrees=12, swing2LimitDegrees=12, twistLimitDegrees=4}
 {bone1Name=dyn_jacket_fr_r_03, bone2Name=dyn_jacket_fr_r_02, swing1LimitDegrees=20, swing2LimitDegrees=20, twistLimitDegrees=4}
 {bone1Name=dyn_jacket_fr_r_04, bone2Name=dyn_jacket_fr_r_03, swing1LimitDegrees=12, swing2LimitDegrees=32, twistLimitDegrees=4}
 {bone1Name=dyn_jacket_fr_r_05, bone2Name=dyn_jacket_fr_r_04, swing1LimitDegrees=12, swing2LimitDegrees=32, twistLimitDegrees=8}
 {bone1Name=dyn_jacket_fr_r_06, bone2Name=dyn_jacket_fr_r_05, swing1LimitDegrees=12, swing2LimitDegrees=32, twistLimitDegrees=8}
 {
  bone1Name=dyn_jacket_fr_l_04
  bone2Name=dyn_jacket_l_03
  swing1LimitDegrees=45
  swing1Motion=Free
  swing2LimitDegrees=45
  swing2Motion=Free
  twistLimitDegrees=45
  twistMotion=Free
 }
 {
  bone1Name=dyn_jacket_l_04
  bone2Name=dyn_jacket_mid_l_04
  swing1LimitDegrees=45
  swing1Motion=Free
  swing2LimitDegrees=45
  swing2Motion=Free
  twistLimitDegrees=45
  twistMotion=Free
 }
 {
  bone1Name=dyn_jacket_l_02
  bone2Name=dyn_jacket_mid_l_02
  swing1LimitDegrees=45
  swing1Motion=Free
  swing2LimitDegrees=45
  swing2Motion=Free
  twistLimitDegrees=45
  twistMotion=Free
 }
 {
  bone1Name=dyn_jacket_mid_l_05
  bone2Name=dyn_jacket_bk_l_02
  swing1LimitDegrees=45
  swing1Motion=Free
  swing2LimitDegrees=45
  swing2Motion=Free
  twistLimitDegrees=45
  twistMotion=Free
 }
 {
  bone1Name=dyn_jacket_mid_l_03
  bone2Name=dyn_jacket_bk_03
  swing1LimitDegrees=45
  swing1Motion=Free
  swing2LimitDegrees=45
  swing2Motion=Free
  twistLimitDegrees=45
  twistMotion=Free
 }
 {
  bone1Name=dyn_jacket_mid_l_02
  bone2Name=dyn_jacket_bk_02
  swing1LimitDegrees=45
  swing1Motion=Free
  swing2LimitDegrees=45
  swing2Motion=Free
  twistLimitDegrees=45
  twistMotion=Free
 }
 {
  bone1Name=dyn_jacket_fr_r_04
  bone2Name=dyn_jacket_r_03
  swing1LimitDegrees=45
  swing1Motion=Free
  swing2LimitDegrees=45
  swing2Motion=Free
  twistLimitDegrees=45
  twistMotion=Free
 }
 {
  bone1Name=dyn_jacket_r_05
  bone2Name=dyn_jacket_mid_r_05
  swing1LimitDegrees=45
  swing1Motion=Free
  swing2LimitDegrees=45
  swing2Motion=Free
  twistLimitDegrees=45
  twistMotion=Free
 }
 {
  bone1Name=dyn_jacket_r_04
  bone2Name=dyn_jacket_mid_r_04
  swing1LimitDegrees=45
  swing1Motion=Free
  swing2LimitDegrees=45
  swing2Motion=Free
  twistLimitDegrees=45
  twistMotion=Free
 }
 {
  bone1Name=dyn_jacket_r_02
  bone2Name=dyn_jacket_mid_r_02
  swing1LimitDegrees=45
  swing1Motion=Free
  swing2LimitDegrees=45
  swing2Motion=Free
  twistLimitDegrees=45
  twistMotion=Free
 }
 {
  bone1Name=dyn_jacket_mid_r_05
  bone2Name=dyn_jacket_bk_r_02
  swing1LimitDegrees=45
  swing1Motion=Free
  swing2LimitDegrees=45
  swing2Motion=Free
  twistLimitDegrees=45
  twistMotion=Free
 }
 {
  bone1Name=dyn_jacket_mid_r_03
  bone2Name=dyn_jacket_bk_03
  swing1LimitDegrees=45
  swing1Motion=Free
  swing2LimitDegrees=45
  swing2Motion=Free
  twistLimitDegrees=45
  twistMotion=Free
 }
 {
  bone1Name=dyn_jacket_mid_r_02
  bone2Name=dyn_jacket_bk_02
  swing1LimitDegrees=45
  swing1Motion=Free
  swing2LimitDegrees=45
  swing2Motion=Free
  twistLimitDegrees=45
  twistMotion=Free
 }
 {
  bone1Name=dyn_jacket_l_05
  bone2Name=dyn_jacket_mid_l_05
  swing1LimitDegrees=45
  swing1Motion=Free
  swing2LimitDegrees=45
  swing2Motion=Free
  twistLimitDegrees=45
  twistMotion=Free
 }
 {
  bone1Name=dyn_jacket_fr_r_06
  bone2Name=dyn_jacket_r_05
  swing1LimitDegrees=45
  swing1Motion=Free
  swing2LimitDegrees=45
  swing2Motion=Free
  twistLimitDegrees=45
  twistMotion=Free
 }
 {
  bone1Name=dyn_jacket_fr_r_03
  bone2Name=dyn_jacket_r_02
  swing1LimitDegrees=45
  swing1Motion=Free
  swing2LimitDegrees=45
  swing2Motion=Free
  twistLimitDegrees=45
  twistMotion=Free
 }
 {
  bone1Name=dyn_jacket_fr_l_03
  bone2Name=dyn_jacket_l_02
  swing1LimitDegrees=45
  swing1Motion=Free
  swing2LimitDegrees=45
  swing2Motion=Free
  twistLimitDegrees=45
  twistMotion=Free
 }
 {
  bone1Name=dyn_jacket_fr_r_05
  bone2Name=dyn_jacket_r_04
  swing1LimitDegrees=45
  swing1Motion=Free
  swing2LimitDegrees=45
  swing2Motion=Free
  twistLimitDegrees=45
  twistMotion=Free
 }
 {
  bone1Name=dyn_jacket_fr_l_06
  bone2Name=dyn_jacket_l_05
  swing1LimitDegrees=45
  swing1Motion=Free
  swing2LimitDegrees=45
  swing2Motion=Free
  twistLimitDegrees=45
  twistMotion=Free
 }
 {
  bone1Name=dyn_jacket_fr_l_05
  bone2Name=dyn_jacket_l_04
  swing1LimitDegrees=45
  swing1Motion=Free
  swing2LimitDegrees=45
  swing2Motion=Free
  twistLimitDegrees=45
  twistMotion=Free
 }
]
