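# Post-process style AnimBlueprint for the AlmondSplash character body, adding secondary motion to hood, necklace and jacket.
# Copies the parent mesh pose through a Control Rig, then AnimDynamics on hood and necklace bones.
# Outputs that pose when cloth is enabled, otherwise adds RigidBody simulation and jacket bone offsets.
# EventGraph drives Chaos cloth and gravity through deprecated state functions, adjusts simulation by movement state, and resets cloth around ziplines.
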AnimBlueprint /BRCosmetics/Characters/Player/Male/Medium/Bodies/M_MED_AlmondSplash/Meshes/M_MED_AlmondSplash_AnimBP
target_skeleton=&/M_MED_AlmondSplash_Skeleton
graphs [
 AnimGraph {nodes [
   Output Pose(Result=Blend Poses by bool(BlendPose_0=Use cached pose 'Main Cache'(), BlendPose_1=Use cached pose 'Rigid Body'(), bActiveValue=Get bClothEnabled()) {blendType=Linear}) {layerGroup=None}
   n1 = Control Rig!(
    Source=Copy Pose From Mesh() {bUseAttachedParent=true}
    Alpha
    isCloth:bool=Get bClothEnabled()
    isWearingCape:bool=Get isWearingCape() -> (isWearingCape:bool)
    isCrouching:bool=Get bIsCrouching()
    isCrouchMoving:bool=Get bIsCrouchSprinting()
    isRailGrinding:bool=Get bIsOnGrindRail()
    isWearingPack:bool=Get isWearingPack() -> (isWearingPack:bool)
    isFrontEnd:bool=Get bIsInFrontEnd()
   ) -> (Pose) {
    alphaBoolBlend {blendOption=Linear}
    controlRigAssetReference {blueprintRigClass=&/M_MED_AlmondSplash_CtrlRig.M_MED_AlmondSplash_CtrlRig_C}
   }
   n2 = Anim Dynamics - Bone: dyn_hood_bk_01!(ComponentPose=Local To Component(LocalPose=@n1), Alpha=1, GravityOverride:Vector="0,0,0", ExternalForce="0,0,0") -> (Pose) {
    angularDampingOverride=0.9999
    angularSpringConstant=1
    bOverrideAngularDamping=true
    bUseGravityOverride=true
    simulationSpace=RootRelative
    alphaBoolBlend {blendOption=Linear}
    alphaScaleBiasClamp {bInterpResult=true, interpSpeedDecreasing=20, interpSpeedIncreasing=20}
    boundBone {boneName=dyn_hood_bk_01}
    componentAppliedLinearAccClamp=<5000 5000 5000>
    componentLinearAccScale=<0.3 0.3 0.3>
    componentLinearVelScale=<0.4 0.4 0.4>
    physicsBodyDefinitions [{
      boundBone {boneName=dyn_hood_bk_01}
      boxExtents=<10 10 10>
      constraintSetup {angularLimitsMax=<0 0 4>, angularLimitsMin=<0 -30 -4>, angularTarget=<1 0 0>}
      localJointOffset=<1 0 0>
     }]
   }
   n3 = Anim Dynamics - Bone: dyn_hood_bk_02!(ComponentPose=@n2, Alpha=1, GravityOverride:Vector="0,0,0", ExternalForce="0,0,0") -> (Pose) {
    angularDampingOverride=0.9999
    angularSpringConstant=1
    bOverrideAngularDamping=true
    bUseGravityOverride=true
    simulationSpace=RootRelative
    alphaBoolBlend {blendOption=Linear}
    alphaScaleBiasClamp {bInterpResult=true, interpSpeedDecreasing=20, interpSpeedIncreasing=20}
    boundBone {boneName=dyn_hood_bk_02}
    componentAppliedLinearAccClamp=<5000 5000 5000>
    componentLinearAccScale=<0.3 0.3 0.3>
    componentLinearVelScale=<0.4 0.4 0.4>
    physicsBodyDefinitions [{
      boundBone {boneName=dyn_hood_bk_02}
      boxExtents=<10 10 10>
      constraintSetup {angularLimitsMax=<0 0 4>, angularLimitsMin=<0 -30 -4>, angularTarget=<1 0 0>}
      localJointOffset=<1 0 0>
     }]
   }
   n4 = Anim Dynamics - Bone: dyn_hood_bk_03!(ComponentPose=@n3, Alpha=1, GravityOverride:Vector="0,0,0", ExternalForce="0,0,0") -> (Pose) {
    angularDampingOverride=0.9999
    angularSpringConstant=1
    bOverrideAngularDamping=true
    bUseGravityOverride=true
    simulationSpace=RootRelative
    alphaBoolBlend {blendOption=Linear}
    alphaScaleBiasClamp {bInterpResult=true, interpSpeedDecreasing=20, interpSpeedIncreasing=20}
    boundBone {boneName=dyn_hood_bk_03}
    componentAppliedLinearAccClamp=<5000 5000 5000>
    componentLinearAccScale=<0.3 0.3 0.3>
    componentLinearVelScale=<0.4 0.4 0.4>
    physicsBodyDefinitions [{
      boundBone {boneName=dyn_hood_bk_03}
      boxExtents=<10 10 10>
      constraintSetup {angularLimitsMax=<0 0 4>, angularLimitsMin=<0 -30 -4>, angularTarget=<1 0 0>}
      localJointOffset=<1 0 0>
     }]
   }
   n5 = Anim Dynamics - Bone: dyn_hood_bk_04!(ComponentPose=@n4, Alpha=1, GravityOverride:Vector="0,0,0", ExternalForce="0,0,0") -> (Pose) {
    angularDampingOverride=0.9999
    angularSpringConstant=1
    bOverrideAngularDamping=true
    bUseGravityOverride=true
    simulationSpace=RootRelative
    alphaBoolBlend {blendOption=Linear}
    alphaScaleBiasClamp {bInterpResult=true, interpSpeedDecreasing=20, interpSpeedIncreasing=20}
    boundBone {boneName=dyn_hood_bk_04}
    componentAppliedLinearAccClamp=<5000 5000 5000>
    componentLinearAccScale=<0.3 0.3 0.3>
    componentLinearVelScale=<0.4 0.4 0.4>
    physicsBodyDefinitions [{
      boundBone {boneName=dyn_hood_bk_04}
      boxExtents=<10 10 10>
      constraintSetup {angularLimitsMax=<0 0 4>, angularLimitsMin=<0 -30 -4>, angularTarget=<1 0 0>}
      localJointOffset=<1 0 0>
     }]
   }
   n6 = Anim Dynamics - Bone: dyn_hood_fr_l_01!(ComponentPose=@n5, Alpha=1, GravityOverride:Vector="0,0,0", ExternalForce="0,0,0") -> (Pose) {
    angularDampingOverride=0.9999
    angularSpringConstant=6000
    bAngularSpring=true
    bOverrideAngularDamping=true
    bUseGravityOverride=true
    simulationSpace=RootRelative
    alphaBoolBlend {blendOption=Linear}
    alphaScaleBiasClamp {bInterpResult=true, interpSpeedDecreasing=20, interpSpeedIncreasing=20}
    boundBone {boneName=dyn_hood_fr_l_01}
    componentAppliedLinearAccClamp=<5000 5000 5000>
    componentLinearAccScale=<0.3 0.3 0.3>
    componentLinearVelScale=<0.4 0.4 0.4>
    physicsBodyDefinitions [{
      boundBone {boneName=dyn_hood_fr_l_01}
      boxExtents=<20 20 20>
      constraintSetup {angularLimitsMax=<0 0 6>, angularLimitsMin=<0 -10 -6>, angularTarget=<1 0 0>}
      localJointOffset=<1 0 0>
     }]
   }
   n7 = Anim Dynamics - Bone: dyn_hood_l_01!(ComponentPose=@n6, Alpha=1, GravityOverride:Vector="0,0,0", ExternalForce="0,0,0") -> (Pose) {
    angularDampingOverride=0.9999
    angularSpringConstant=6000
    bAngularSpring=true
    bOverrideAngularDamping=true
    bUseGravityOverride=true
    simulationSpace=RootRelative
    alphaBoolBlend {blendOption=Linear}
    alphaScaleBiasClamp {bInterpResult=true, interpSpeedDecreasing=20, interpSpeedIncreasing=20}
    boundBone {boneName=dyn_hood_l_01}
    componentAppliedLinearAccClamp=<5000 5000 5000>
    componentLinearAccScale=<0.3 0.3 0.3>
    componentLinearVelScale=<0.4 0.4 0.4>
    physicsBodyDefinitions [{
      boundBone {boneName=dyn_hood_l_01}
      boxExtents=<20 20 20>
      constraintSetup {angularLimitsMax=<0 0 6>, angularLimitsMin=<0 -10 -6>, angularTarget=<1 0 0>}
      localJointOffset=<1 0 0>
     }]
   }
   n8 = Anim Dynamics - Bone: dyn_hood_mid_l_01!(ComponentPose=@n7, Alpha=1, GravityOverride:Vector="0,0,0", ExternalForce="0,0,0") -> (Pose) {
    angularDampingOverride=0.9999
    angularSpringConstant=6000
    bAngularSpring=true
    bOverrideAngularDamping=true
    bUseGravityOverride=true
    simulationSpace=RootRelative
    alphaBoolBlend {blendOption=Linear}
    alphaScaleBiasClamp {bInterpResult=true, interpSpeedDecreasing=20, interpSpeedIncreasing=20}
    boundBone {boneName=dyn_hood_mid_l_01}
    componentAppliedLinearAccClamp=<5000 5000 5000>
    componentLinearAccScale=<0.3 0.3 0.3>
    componentLinearVelScale=<0.4 0.4 0.4>
    physicsBodyDefinitions [{
      boundBone {boneName=dyn_hood_mid_l_01}
      boxExtents=<20 20 20>
      constraintSetup {angularLimitsMax=<0 0 6>, angularLimitsMin=<0 -10 -6>, angularTarget=<1 0 0>}
      localJointOffset=<1 0 0>
     }]
   }
   n9 = Anim Dynamics - Bone: dyn_hood_c_01!(ComponentPose=@n8, Alpha=1, GravityOverride:Vector="0,0,0", ExternalForce="0,0,0") -> (Pose) {
    angularDampingOverride=0.9999
    angularSpringConstant=6000
    bAngularSpring=true
    bOverrideAngularDamping=true
    bUseGravityOverride=true
    simulationSpace=RootRelative
    alphaBoolBlend {blendOption=Linear}
    alphaScaleBiasClamp {bInterpResult=true, interpSpeedDecreasing=20, interpSpeedIncreasing=20}
    boundBone {boneName=dyn_hood_c_01}
    componentAppliedLinearAccClamp=<5000 5000 5000>
    componentLinearAccScale=<0.3 0.3 0.3>
    componentLinearVelScale=<0.4 0.4 0.4>
    physicsBodyDefinitions [{
      boundBone {boneName=dyn_hood_c_01}
      boxExtents=<20 20 20>
      constraintSetup {angularLimitsMax=<0 0 6>, angularLimitsMin=<0 -10 -6>, angularTarget=<1 0 0>}
      localJointOffset=<1 0 0>
     }]
   }
   n10 = Anim Dynamics - Bone: dyn_hood_mid_r_01!(ComponentPose=@n9, Alpha=1, GravityOverride:Vector="0,0,0", ExternalForce="0,0,0") -> (Pose) {
    angularDampingOverride=0.9999
    angularSpringConstant=6000
    bAngularSpring=true
    bOverrideAngularDamping=true
    bUseGravityOverride=true
    simulationSpace=RootRelative
    alphaBoolBlend {blendOption=Linear}
    alphaScaleBiasClamp {bInterpResult=true, interpSpeedDecreasing=20, interpSpeedIncreasing=20}
    boundBone {boneName=dyn_hood_mid_r_01}
    componentAppliedLinearAccClamp=<5000 5000 5000>
    componentLinearAccScale=<0.3 0.3 0.3>
    componentLinearVelScale=<0.4 0.4 0.4>
    physicsBodyDefinitions [{
      boundBone {boneName=dyn_hood_mid_r_01}
      boxExtents=<20 20 20>
      constraintSetup {angularLimitsMax=<0 0 6>, angularLimitsMin=<0 -10 -6>, angularTarget=<1 0 0>}
      localJointOffset=<1 0 0>
     }]
   }
   n11 = Anim Dynamics - Bone: dyn_hood_r_01!(ComponentPose=@n10, Alpha=1, GravityOverride:Vector="0,0,0", ExternalForce="0,0,0") -> (Pose) {
    angularDampingOverride=0.9999
    angularSpringConstant=6000
    bAngularSpring=true
    bOverrideAngularDamping=true
    bUseGravityOverride=true
    simulationSpace=RootRelative
    alphaBoolBlend {blendOption=Linear}
    alphaScaleBiasClamp {bInterpResult=true, interpSpeedDecreasing=20, interpSpeedIncreasing=20}
    boundBone {boneName=dyn_hood_r_01}
    componentAppliedLinearAccClamp=<5000 5000 5000>
    componentLinearAccScale=<0.3 0.3 0.3>
    componentLinearVelScale=<0.4 0.4 0.4>
    physicsBodyDefinitions [{
      boundBone {boneName=dyn_hood_r_01}
      boxExtents=<20 20 20>
      constraintSetup {angularLimitsMax=<0 0 6>, angularLimitsMin=<0 -10 -6>, angularTarget=<1 0 0>}
      localJointOffset=<1 0 0>
     }]
   }
   n12 = Anim Dynamics - Bone: dyn_hood_fr_r_01!(ComponentPose=@n11, Alpha=1, GravityOverride:Vector="0,0,0", ExternalForce="0,0,0") -> (Pose) {
    angularDampingOverride=0.9999
    angularSpringConstant=6000
    bAngularSpring=true
    bOverrideAngularDamping=true
    bUseGravityOverride=true
    simulationSpace=RootRelative
    alphaBoolBlend {blendOption=Linear}
    alphaScaleBiasClamp {bInterpResult=true, interpSpeedDecreasing=20, interpSpeedIncreasing=20}
    boundBone {boneName=dyn_hood_fr_r_01}
    componentAppliedLinearAccClamp=<5000 5000 5000>
    componentLinearAccScale=<0.3 0.3 0.3>
    componentLinearVelScale=<0.4 0.4 0.4>
    physicsBodyDefinitions [{
      boundBone {boneName=dyn_hood_fr_r_01}
      boxExtents=<20 20 20>
      constraintSetup {angularLimitsMax=<0 0 6>, angularLimitsMin=<0 -10 -6>, angularTarget=<1 0 0>}
      localJointOffset=<1 0 0>
     }]
   }
   n13 = Anim Dynamics - Bone: dyn_necklace_01!(ComponentPose=@n12, Alpha=1, GravityOverride:Vector="0,0,0", ExternalForce="0,0,0") -> (Pose) {
    angularDampingOverride=0.99
    angularSpringConstant=1000
    bAngularSpring=true
    bOverrideAngularDamping=true
    bUseGravityOverride=true
    simulationSpace=BoneRelative
    alphaBoolBlend {blendOption=Linear}
    boundBone {boneName=dyn_necklace_01}
    componentAppliedLinearAccClamp=<5000 5000 5000>
    componentLinearAccScale=<0.5 0.5 0.5>
    componentLinearVelScale=<0.6 0.6 0.6>
    physicsBodyDefinitions [{
      boundBone {boneName=dyn_necklace_01}
      boxExtents=<10 10 10>
      constraintSetup {angularLimitsMax=<0 0 10>, angularLimitsMin=<0 0 -10>, angularTarget=<1 0 0>}
      localJointOffset=<0.5 0 0>
     }]
    relativeSpaceBone {boneName=spine_04}
   }
   n14 = Anim Dynamics - Bone: dyn_necklace_02!(ComponentPose=@n13, Alpha=1, GravityOverride:Vector="0,0,0", ExternalForce="0,0,0") -> (Pose) {
    angularDampingOverride=0.99
    angularSpringConstant=1000
    bAngularSpring=true
    bOverrideAngularDamping=true
    bUseGravityOverride=true
    simulationSpace=BoneRelative
    alphaBoolBlend {blendOption=Linear}
    boundBone {boneName=dyn_necklace_02}
    componentAppliedLinearAccClamp=<5000 5000 5000>
    componentLinearAccScale=<0.5 0.5 0.5>
    componentLinearVelScale=<0.6 0.6 0.6>
    physicsBodyDefinitions [{
      boundBone {boneName=dyn_necklace_02}
      boxExtents=<12 12 12>
      constraintSetup {angularLimitsMax=<0 0 10>, angularLimitsMin=<0 0 -10>, angularTarget=<1 0 0>}
      localJointOffset=<0.5 0 0>
     }]
    relativeSpaceBone {boneName=spine_04}
   }
   n15 = Anim Dynamics - Bone: dyn_necklace_03!(ComponentPose=@n14, Alpha=1, GravityOverride:Vector="0,0,0", ExternalForce="0,0,0") -> (Pose) {
    angularDampingOverride=0.99
    angularSpringConstant=1000
    bAngularSpring=true
    bOverrideAngularDamping=true
    bUseGravityOverride=true
    simulationSpace=BoneRelative
    alphaBoolBlend {blendOption=Linear}
    boundBone {boneName=dyn_necklace_03}
    componentAppliedLinearAccClamp=<5000 5000 5000>
    componentLinearAccScale=<0.5 0.5 0.5>
    componentLinearVelScale=<0.6 0.6 0.6>
    physicsBodyDefinitions [{
      boundBone {boneName=dyn_necklace_03}
      boxExtents=<30 30 30>
      constraintSetup {angularLimitsMax=<0 0 4>, angularLimitsMin=<0 0 -4>, angularTarget=<1 0 0>}
      localJointOffset=<0.5 0 0>
     }]
    relativeSpaceBone {boneName=spine_04}
   }
   Save cached pose 'Main Cache'(Pose=Component To Local(ComponentPose=@n15)) {cacheName=Main Cache}
   n16 = RigidBody!(
    ComponentPose=Local To Component(LocalPose=Use cached pose 'Main Cache'())
    Alpha
    OverrideWorldGravity:Vector="0,0,0"
    ComponentLinearAccScale:Vector="0.4,0.4,0.4"
    ComponentLinearVelScale:Vector="0.4,0.4,0.4"
    SimSpaceSettings:Sim Space Settings Structure=(WorldAlpha=0.5,VelocityScaleZ=0,DampingAlpha=1,MaxLinearVelocity=250,MaxAngularVelocity=3,MaxLinearAcceleration=350,MaxAngularAcceleration=500,ExternalLinearDragV=(X=0.1,Y=0.1,Z=0),ExternalLinearVelocity=(X=0,Y=0,Z=0),ExternalAngularVelocity=(X=0,Y=0,Z=0))
   ) -> (Pose) {
    bClampLinearTranslationLimitToRefPose=true
    bOverrideWorldGravity=true
    overridePhysicsAsset=&/M_MED_AlmondSplash_Physics
    simulationSpace=BaseBoneSpace
    alphaBoolBlend {blendOption=Linear}
    baseBoneRef {boneName=pelvis}
    componentAppliedLinearAccClamp=<1000 1000 1000>
    componentLinearAccScale=<0.4 0.4 0.4>
    componentLinearVelScale=<0.4 0.4 0.4>
    overrideWorldGravity {z=-980}
    simSpaceSettings {
     maxAngularAcceleration=500
     maxAngularVelocity=3
     maxLinearAcceleration=350
     maxLinearVelocity=250
     velocityScaleZ=0
     worldAlpha=0.5
     externalLinearDragV=<0.1 0.1>
    }
   }
   n17 = "Transform (Modify) Bone - Bone: dyn_jacket_bk_MSTR"(ComponentPose=@n16, Alpha=1, Translation="0,0,1", Rotation="0,0,0", Scale="1,1,1") {
    translationMode=BMM_Additive
    translationSpace=BCS_BoneSpace
    alphaBoolBlend {blendOption=Linear}
    boneToModify {boneName=dyn_jacket_bk_MSTR}
    translation {z=1}
   }
   n18 = "Transform (Modify) Bone - Bone: dyn_jacket_mid_MSTR_R"(ComponentPose=@n17, Alpha=1, Translation="0,0,1", Rotation="0,0,0", Scale="1,1,1") {
    translationMode=BMM_Additive
    translationSpace=BCS_BoneSpace
    alphaBoolBlend {blendOption=Linear}
    boneToModify {boneName=dyn_jacket_mid_MSTR_R}
    translation {z=1}
   }
   n19 = "Transform (Modify) Bone - Bone: dyn_jacket_MSTR_R"(ComponentPose=@n18, Alpha=1, Translation="0,0,1", Rotation="0,0,0", Scale="1,1,1") {
    translationMode=BMM_Additive
    translationSpace=BCS_BoneSpace
    alphaBoolBlend {blendOption=Linear}
    boneToModify {boneName=dyn_jacket_MSTR_R}
    translation {z=1}
   }
   n20 = "Transform (Modify) Bone - Bone: dyn_jacket_bk_02"(ComponentPose=@n19, Alpha=1, Translation="0,0,0.5", Rotation="0,0,0", Scale="1,1,1") {
    translationMode=BMM_Additive
    translationSpace=BCS_BoneSpace
    alphaBoolBlend {blendOption=Linear}
    boneToModify {boneName=dyn_jacket_bk_02}
    translation {z=0.5}
   }
   n21 = "Transform (Modify) Bone - Bone: dyn_jacket_mid_r_02"(ComponentPose=@n20, Alpha=1, Translation="0,0,0.5", Rotation="0,0,0", Scale="1,1,1") {
    translationMode=BMM_Additive
    translationSpace=BCS_BoneSpace
    alphaBoolBlend {blendOption=Linear}
    boneToModify {boneName=dyn_jacket_mid_r_02}
    translation {z=0.5}
   }
   n22 = "Transform (Modify) Bone - Bone: dyn_jacket_fr_MSTR_L"!(ComponentPose=@n21, bAlphaBoolEnabled:bool=True, Translation="0,0,5", Rotation="0,0,0", Scale="1,1,1") -> (Pose) {
    alphaInputType=Bool
    rotationMode=BMM_Additive
    rotationSpace=BCS_BoneSpace
    translationMode=BMM_Additive
    translationSpace=BCS_BoneSpace
    alphaBoolBlend {blendInTime=0.05, blendOption=Linear, blendOutTime=1}
    boneToModify {boneName=dyn_jacket_fr_MSTR_L}
    translation {z=5}
   }
   n23 = "Transform (Modify) Bone - Bone: dyn_jacket_MSTR_L"!(ComponentPose=@n22, bAlphaBoolEnabled:bool=True, Translation="0,0,5", Rotation="0,0,0", Scale="1,1,1") -> (Pose) {
    alphaInputType=Bool
    rotationMode=BMM_Additive
    rotationSpace=BCS_BoneSpace
    translationMode=BMM_Additive
    translationSpace=BCS_BoneSpace
    alphaBoolBlend {blendInTime=0.05, blendOption=Linear, blendOutTime=1}
    boneToModify {boneName=dyn_jacket_MSTR_L}
    translation {z=5}
   }
   n24 = "Transform (Modify) Bone - Bone: dyn_jacket_mid_MSTR_L"!(ComponentPose=@n23, bAlphaBoolEnabled:bool=True, Translation="0,0,5", Rotation="0,0,0", Scale="1,1,1") -> (Pose) {
    alphaInputType=Bool
    rotationMode=BMM_Additive
    rotationSpace=BCS_BoneSpace
    translationMode=BMM_Additive
    translationSpace=BCS_BoneSpace
    alphaBoolBlend {blendInTime=0.05, blendOption=Linear, blendOutTime=1}
    boneToModify {boneName=dyn_jacket_mid_MSTR_L}
    translation {z=5}
   }
   n25 = "Transform (Modify) Bone - Bone: dyn_jacket_fr_MSTR_R"!(ComponentPose=@n24, bAlphaBoolEnabled:bool=True, Translation="0,0,5", Rotation="0,0,0", Scale="1,1,1") -> (Pose) {
    alphaInputType=Bool
    rotationMode=BMM_Additive
    rotationSpace=BCS_BoneSpace
    translationMode=BMM_Additive
    translationSpace=BCS_BoneSpace
    alphaBoolBlend {blendInTime=0.05, blendOption=Linear, blendOutTime=1}
    boneToModify {boneName=dyn_jacket_fr_MSTR_R}
    translation {z=5}
   }
   n26 = "Transform (Modify) Bone - Bone: dyn_jacket_bk_MSTR"!(ComponentPose=@n25, bAlphaBoolEnabled:bool=True, Translation="0,0,5", Rotation="0,0,0", Scale="1,1,1") -> (Pose) {
    alphaInputType=Bool
    rotationMode=BMM_Additive
    rotationSpace=BCS_BoneSpace
    translationMode=BMM_Additive
    translationSpace=BCS_BoneSpace
    alphaBoolBlend {blendInTime=0.05, blendOption=Linear, blendOutTime=1}
    boneToModify {boneName=dyn_jacket_bk_MSTR}
    translation {z=5}
   }
   n27 = "Transform (Modify) Bone - Bone: dyn_jacket_mid_MSTR_R"!(ComponentPose=@n26, bAlphaBoolEnabled:bool=True, Translation="0,0,5", Rotation="0,0,0", Scale="1,1,1") -> (Pose) {
    alphaInputType=Bool
    rotationMode=BMM_Additive
    rotationSpace=BCS_BoneSpace
    translationMode=BMM_Additive
    translationSpace=BCS_BoneSpace
    alphaBoolBlend {blendInTime=0.05, blendOption=Linear, blendOutTime=1}
    boneToModify {boneName=dyn_jacket_mid_MSTR_R}
    translation {z=5}
   }
   n28 = "Transform (Modify) Bone - Bone: dyn_jacket_fr_MSTR_L"(ComponentPose=@n27, bAlphaBoolEnabled=True, Translation="0,0,0", Rotation="6,0,0", Scale="1,1,1") {
    alphaInputType=Bool
    rotationMode=BMM_Additive
    rotationSpace=BCS_BoneSpace
    translationMode=BMM_Additive
    translationSpace=BCS_BoneSpace
    alphaBoolBlend {blendInTime=0.05, blendOption=Linear, blendOutTime=0.5}
    boneToModify {boneName=dyn_jacket_fr_MSTR_L}
    rotation {pitch=6}
   }
   n29 = "Transform (Modify) Bone - Bone: dyn_jacket_fr_MSTR_R"(ComponentPose=@n28, bAlphaBoolEnabled=True, Translation="0,0,0", Rotation="6,0,0", Scale="1,1,1") {
    alphaInputType=Bool
    rotationMode=BMM_Additive
    rotationSpace=BCS_BoneSpace
    translationMode=BMM_Additive
    translationSpace=BCS_BoneSpace
    alphaBoolBlend {blendInTime=0.05, blendOption=Linear, blendOutTime=0.5}
    boneToModify {boneName=dyn_jacket_fr_MSTR_R}
    rotation {pitch=6}
   }
   n30 = "Transform (Modify) Bone - Bone: dyn_jacket_mid_MSTR_L"(ComponentPose=@n29, bAlphaBoolEnabled=True, Translation="0,0,0", Rotation="6,0,0", Scale="1,1,1") {
    alphaInputType=Bool
    rotationMode=BMM_Additive
    rotationSpace=BCS_BoneSpace
    translationMode=BMM_Additive
    translationSpace=BCS_BoneSpace
    alphaBoolBlend {blendInTime=0.05, blendOption=Linear, blendOutTime=0.5}
    boneToModify {boneName=dyn_jacket_mid_MSTR_L}
    rotation {pitch=6}
   }
   n31 = "Transform (Modify) Bone - Bone: dyn_jacket_mid_MSTR_R"(ComponentPose=@n30, bAlphaBoolEnabled=True, Translation="0,0,0", Rotation="6,0,0", Scale="1,1,1") {
    alphaInputType=Bool
    rotationMode=BMM_Additive
    rotationSpace=BCS_BoneSpace
    translationMode=BMM_Additive
    translationSpace=BCS_BoneSpace
    alphaBoolBlend {blendInTime=0.05, blendOption=Linear, blendOutTime=0.5}
    boneToModify {boneName=dyn_jacket_mid_MSTR_R}
    rotation {pitch=6}
   }
   n32 = "Transform (Modify) Bone - Bone: dyn_jacket_bk_MSTR"(ComponentPose=@n31, bAlphaBoolEnabled=True, Translation="0,0,0", Rotation="6,0,0", Scale="1,1,1") {
    alphaInputType=Bool
    rotationMode=BMM_Additive
    rotationSpace=BCS_BoneSpace
    translationMode=BMM_Additive
    translationSpace=BCS_BoneSpace
    alphaBoolBlend {blendInTime=0.05, blendOption=Linear, blendOutTime=0.5}
    boneToModify {boneName=dyn_jacket_bk_MSTR}
    rotation {pitch=6}
   }
   Save cached pose 'Rigid Body'(Pose=Component To Local(ComponentPose=@n32)) {cacheName=Rigid Body}
  ]}
 EventGraph {nodes [
   Event BlueprintUpdateAnimation()
   > n1 = Sequence!(execute) -> (then_0, then_1, then_2:Exec, then_3:Exec, then_4:Exec, then_5:Exec, then_6:Exec)
   > Set DeltaTimeLocal(DeltaTimeLocal:double=0) -> (Output_Get:double)
   > Branch(Condition=Get bClothEnabled())
   > SetClothingInteractorIfNotSet()
   > Cloth Controller Function DEPRECATED!(
    execute
    CharacterPartAnimInstance:Custom Character Part Anim Instance Object Reference=Self-Reference()
    ClothingAssetName:String=M_MED_AlmondSplash_Clothing_0
    Cloth Paramaters:Cloth Params Struct Structure=Get Cloth Paramaters() -> (Cloth Paramaters:Cloth Params Struct Structure)
    Cloth Core Settings:Cloth Core Setting Struct Structure=Get Cloth Core Settings() -> (Cloth Core Settings:Cloth Core Setting Struct Structure)
    ClothUpAxis:Vector="1,0,0"
    OnGround_Moving_SpeedLimit:double=2
    OnGround_Moving_AngularVel_Max:double=0.05
    OnGround_Moving_FictitiousVel_Max:double=0.25
    LinVelScale_FwdBwd_Clamp:double=0
    LinVelScale_UpDwn_Clamp:double=0.7
    DeltaTime:double=Get DeltaTimeLocal() -> (DeltaTimeLocal:double)
    GameTimeInSeconds:double=GetGameTimeInSeconds()
    WindGustsParams:Wind Gust Params Struct Structure=Get Wind Gusts Params() -> (Wind Gusts Params:Wind Gust Params Struct Structure)
    WindGustLocalSettings:Wind Gust Local Setting Struct Structure=Get Wind Gust Local Settings() -> (Wind Gust Local Settings:Wind Gust Local Setting Struct Structure)
   ) -> (
    then
    CurrentStateOut:String
    Cloth Settings Out:Cloth Setting Struct Structure
    Cloth Core Settings Out:Cloth Core Setting Struct Structure
    WindGustSettings Out:Wind Gust Setting Struct Structure
    WindGustLocalSettings Out:Wind Gust Local Setting Struct Structure
    GustStrength - WindVelocityOut:Vector
    LinearVelVector-DELETE LATER:Vector
   )
   n2 = BreakRotator(InRot=Get PawnRotationRate() -> (PawnRotationRate:Rotator))
   > Branch(Condition="float > float"(A="Absolute (Float)"(A=@n2.Yaw), B=5) {is=title K2Node_CallFunction})
   > SetClothingInteractorIfNotSet()
   > SetVelocityScale!(
    execute
    self:Chaos Clothing Interactor Object Reference=Get ChaosClothingInteractor() -> (ChaosClothingInteractor:Chaos Clothing Interactor Object Reference)
    LinearVelocityScale:Vector="0.025,0.025,0.025"
    AngularVelocityScale:float=0
    FictitiousAngularScale:float=0
   ) -> (then)
   n3 = Gravity Override State Function DEPRECATED(
    execute=@n1.then_1
    CharacterPartAnimInstance=Self-Reference()
    GameTimeInSeconds=GetGameTimeInSeconds()
    GravityOverrideParams=Get Gravity Override Params()
    RBAN Params=Get RBAN Params()
   )
   > Set Gravity Override(Gravity Override=@n3)
   > Set RBAN Settings Out(RBAN Settings Out:RBANSetting Struct Structure=@n3.RBAN SettingsOut) -> (Output_Get:RBANSetting Struct Structure)
   n4 = Branch(execute=@n1.then_2, Condition=Get bIsInFrontEnd())
   > Set Simulation Space(Simulation Space:AnimPhysSimSpaceType Enum=World) -> (Output_Get:AnimPhysSimSpaceType Enum)
   > Set AnimDynAlpha(AnimDynAlpha:double=0.35) -> (Output_Get:double)
   Set Simulation Space(execute=@n4.else)
   > Set AnimDynAlpha(AnimDynAlpha=1)
   n5 = Branch(execute=@n1.then_3, Condition=OR Boolean(A=Get isWearingPack(), B=Get IsOnZipline() -> (IsOnZipline:bool)))
   > Set AnimDynAlpha(AnimDynAlpha=0)
   n6 = Get bIsJumping()
   n7 = Get bIsTacticalSprinting()
   n8 = Branch(
    execute=@n5.else
    Condition=OR Boolean!(A=Get bIsTargeting(), B=Get bIsTacticalSprinting(), C:bool=AND Boolean(A=@n7, B=@n6), D:bool=Get bIsSkyDiving(), E:bool) -> (ReturnValue)
   )
   > Set AnimDynAlpha(AnimDynAlpha=0.4)
   Set AnimDynAlpha(execute=@n8.else, AnimDynAlpha=1)
   n9 = Branch(execute=@n1.then_4, Condition=AND Boolean(A=Get bIsPlayingMeleeAnim(), B=Get bIsCrouching()))
   > Set CrouchMelee(CrouchMelee:bool=true) -> (Output_Get:bool)
   Set CrouchMelee(execute=@n9.else, CrouchMelee=false)
   Event OnCharacterCustomizationLoaded()
   n10 = CheckForBackbling()
   > Set isWearingPack(isWearingPack:bool=OR Boolean!(A=@n10, B=@n10.bIsBackblingACape, C=@n10.bIsBackblingLong) -> (ReturnValue)) -> (Output_Get:bool)
   > Set isWearingCape(isWearingCape=@n10.bIsBackblingLong)
   > SetClothingInteractorIfNotSet()
   Event OnEndFalling()
   > SetClothingInteractorIfNotSet()
   > Branch(Condition=Get resetClothLandingFromZilpine() -> (resetClothLandingFromZilpine:bool))
   n11 = Get ChaosClothingInteractor()
   > ResetAndTeleport!(execute, self:Chaos Clothing Interactor Object Reference=@n11, bReset:bool=true, bTeleport:bool=true) -> (then)
   > Set resetClothLandingFromZilpine(resetClothLandingFromZilpine:bool=false) -> (Output_Get:bool)
   n12 = Event BlueprintProcessAnimInputEvent()
   n13 = IsNamed(self=@n12.AnimInputEvent, InEventName=aie.Player.Zipline.Entered)
   > n14 = Branch(Condition=@n13)
   > Set IsOnZipline(IsOnZipline:bool=true) -> (Output_Get:bool)
   > Branch(Condition=Get bClothEnabled())
   > SetClothingInteractorIfNotSet()
   > n15 = Branch(Condition=@n13)
   > Delay(Duration=0.2)
   > ResetAndTeleport(self=@n11, bReset=true, bTeleport=true)
   n16 = IsNamed(self=@n12.AnimInputEvent, InEventName=aie.Player.Zipline.Exited)
   Branch(execute=@n15.else, Condition=@n16)
   > Set resetClothLandingFromZilpine(resetClothLandingFromZilpine=true)
   > Delay(Duration=1.1)
   Branch(execute=@n14.else, Condition=@n16)
   > Set IsOnZipline(IsOnZipline=false)
   Get Cloth Paramaters_0() -> (Cloth Paramaters_0:Cloth Params Struct Structure)
  ]}
 SetClothingInteractorIfNotSet {nodes [
   SetClothingInteractorIfNotSet() {is=title K2Node_FunctionEntry}
   n1 = IsValid(Object=Get ChaosClothingInteractor())
   > Branch(Condition=AND Boolean(A=Get bClothEnabled(), B=NOT Boolean(A=@n1)))
   n2 = MakeLiteralString(Value=M_MED_AlmondSplash_Clothing_0)
   n3 = GetOwningComponent()
   > n4 = Cast To ChaosClothingInteractor!(
    execute
    Object=GetClothingInteractor!(
     self:Clothing Simulation Interactor Object Reference=GetClothingSimulationInteractor!(self:Skeletal Mesh Component Object Reference=@n3, ClothingAssetName:Name=None) -> (ReturnValue:Clothing Simulation Interactor Object Reference)
     ClothingAssetName:Name=String To Name(InString=@n2)
    ) -> (ReturnValue:Clothing Interactor Object Reference)
   ) -> (then, CastFailed, AsChaos Clothing Interactor:Chaos Clothing Interactor Object Reference) {is=title K2Node_DynamicCast}
   > Set ChaosClothingInteractor(ChaosClothingInteractor:Chaos Clothing Interactor Object Reference=@n4) -> (Output_Get:Chaos Clothing Interactor Object Reference)
  ]}
]
variables [
 AnimDynAlpha {type=number}
 ChainConstraintAlpha {type=number, default=0.2}
 ChaosClothingInteractor {type=ChaosClothingInteractor, default=null}
 Cloth Core Settings {
  type=ClothCoreSettingStruct
  default {
   collisionThickness=1.5
   dampingCoefficient=0.06
   frictionCoefficient=0.3
   localDampingCoefficient=0
   numberIterations=2
   numberSubsteps=1
   selfCollisionThickness=0
   animDriveDamping=<0 1>
   animDriveStiffness=<0 1>
   areaStiffness=<1 1>
   bendingStiffness=<0.5 1>
   edgeStiffness=<0.85 1>
   tetherStiffness=<1 1>
  }
 }
 Cloth Paramaters {
  type=ClothParamsStruct
  default {
   dBNO {
    angularVelocityScale=0.6
    fictitiousAngularScale=0.1
    jointName=pelvis
    dragCoefficient=<0.2 1>
    gravityOverride=<-980 400 0>
    liftCoefficient=<0 0.5>
    linearVelocityScale=<0.3 0.3 0.1>
   }
   dBNO_Carried {
    angularVelocityScale=0.6
    fictitiousAngularScale=0.1
    jointName=pelvis
    dragCoefficient=<0.2 1>
    gravityOverride=<-980 -200 0>
    liftCoefficient=<0 0.5>
    linearVelocityScale=<0.3 0.3 0.1>
   }
   dBNO_Impulse {
    angularVelocityScale=0
    fictitiousAngularScale=0
    jointName=pelvis
    dragCoefficient=<0 0.2>
    gravityOverride=<-3000 -2000 0>
    liftCoefficient=<0 0.1>
    linearVelocityScale=<0.3 0.1 0.3>
   }
   dBNO_RollBwd {
    angularVelocityScale=0
    fictitiousAngularScale=0
    jointName=pelvis
    dragCoefficient=<0 0.2>
    gravityOverride=<-2000 0 0>
    liftCoefficient=<0 0.1>
    linearVelocityScale=<0.3 0.1 0.3>
   }
   dBNO_RollFwd {
    angularVelocityScale=0
    fictitiousAngularScale=0
    jointName=pelvis
    dragCoefficient=<0 0.2>
    gravityOverride=<-980 0 0>
    liftCoefficient=<0 0.1>
    linearVelocityScale=<0.3 0.1 0.3>
   }
   dBNO_Tumbling {
    angularVelocityScale=0
    fictitiousAngularScale=0
    jointName=pelvis
    dragCoefficient=<0 0.2>
    gravityOverride=<-2500 -300 0>
    liftCoefficient=<0 0.1>
    linearVelocityScale=<0.3 0.1 0.3>
   }
   emoteOrMelee {
    angularVelocityScale=0.2
    fictitiousAngularScale=0
    jointName=pelvis
    dragCoefficient=<0 0.01>
    gravityOverride=<-980 0 0>
    liftCoefficient=<0 0.01>
    linearVelocityScale=<0.2 0.2 0.2>
   }
   front_End {
    angularVelocityScale=0.1
    fictitiousAngularScale=0.1
    jointName=root
    dragCoefficient=<0.2 0.2>
    gravityOverride=<0 0 -980>
    liftCoefficient=<0.1 0.1>
    linearVelocityScale=<0.1 0.1 0.1>
   }
   onGround_Crouch_Moving {
    angularVelocityScale=0.05
    fictitiousAngularScale=0
    jointName=root
    dragCoefficient=<0.01 0.05>
    gravityOverride=<0 0 -980>
    liftCoefficient=<0.01 0.05>
    linearVelocityScale=<0.4 0.4 0.5>
   }
   onGround_Falling {
    angularVelocityScale=0.1
    fictitiousAngularScale=0
    jointName=root
    dragCoefficient=<0.01 0.02>
    gravityOverride=<0 -400 -1500>
    liftCoefficient=<0.01 0.03>
    linearVelocityScale=<0.2 0.2 0.2>
   }
   onGround_Moving {
    angularVelocityScale=0.1
    fictitiousAngularScale=0
    jointName=root
    dragCoefficient=<0.1 0.4>
    gravityOverride=<0 0 -980>
    liftCoefficient=<0.1 0.5>
    linearVelocityScale=<0.3 0.2 0.3>
   }
   onGround_Standing {
    angularVelocityScale=0.3
    fictitiousAngularScale=0.1
    jointName=root
    dragCoefficient=<0 1>
    gravityOverride=<0 0 -980>
    liftCoefficient=<0 0.5>
    linearVelocityScale=<0.4 0.4 0.4>
   }
   rail_Grinding {
    angularVelocityScale=0.2
    fictitiousAngularScale=0
    jointName=root
    dragCoefficient=<0 0.05>
    gravityOverride=<300 -100 -800>
    liftCoefficient=<0 0.1>
    linearVelocityScale=<0.1 0.1 0.1>
   }
   riding_Moving {
    angularVelocityScale=0.05
    fictitiousAngularScale=1
    jointName=root
    dragCoefficient=<0.5 0.5>
    gravityOverride=<0 0 -980>
    liftCoefficient=<0.5 0.5>
    linearVelocityScale=<0.4 0.6 0.3>
   }
   riding_Moving_Targeting {
    angularVelocityScale=0.05
    fictitiousAngularScale=0.7
    jointName=root
    dragCoefficient=<0.5 0.5>
    gravityOverride=<0 0 -980>
    liftCoefficient=<0.5 0.5>
    linearVelocityScale=<0.2 0.6 0.2>
   }
   riding_Standing {
    angularVelocityScale=0.05
    fictitiousAngularScale=1
    jointName=root
    dragCoefficient=<0.5 0.5>
    gravityOverride=<0 -450 -980>
    liftCoefficient=<0.5 0.5>
    linearVelocityScale=<0.6 0.6 0.3>
   }
   skydive_DivingDown {
    angularVelocityScale=0.05
    fictitiousAngularScale=0
    jointName=pelvis
    dragCoefficient=<0.05 0.05>
    gravityOverride=<-450 -100 0>
    liftCoefficient=<0.05 0.05>
    linearVelocityScale=<0.1 0.1 0.1>
   }
   skydive_FloatingDown {
    angularVelocityScale=0.01
    fictitiousAngularScale=0
    jointName=root
    dragCoefficient=<0 0.01>
    gravityOverride=<0 0 0>
    liftCoefficient=<0 0.01>
    linearVelocityScale=<0.05 0.05 0.05>
   }
   skydive_GoingUp {
    angularVelocityScale=0.01
    fictitiousAngularScale=0
    jointName=root
    dragCoefficient=<0 0.01>
    gravityOverride=<0 0 0>
    liftCoefficient=<0 0.01>
    linearVelocityScale=<0.05 0.05 0.05>
   }
   skydive_Parachuting {
    angularVelocityScale=0.025
    fictitiousAngularScale=0
    jointName=root
    dragCoefficient=<0 0.01>
    gravityOverride=<0 0 -980>
    liftCoefficient=<0 0.01>
    linearVelocityScale=<0.085 0.085 0.085>
   }
   skydive_Surfing {
    angularVelocityScale=0.01
    fictitiousAngularScale=0
    jointName=root
    dragCoefficient=<0 0.01>
    gravityOverride=<0 0 -600>
    liftCoefficient=<0 0.01>
    linearVelocityScale=<0.1 0.1 0.05>
   }
   swimming {
    angularVelocityScale=0.6
    fictitiousAngularScale=0
    jointName=root
    dragCoefficient=<0 0.05>
    gravityOverride=<0 -600 100>
    liftCoefficient=<0 0.1>
    linearVelocityScale=<0.6 0.2 0.2>
   }
   tactical_Sprint {
    angularVelocityScale=0.01
    fictitiousAngularScale=0
    jointName=root
    dragCoefficient=<0.1 0.2>
    gravityOverride=<0 -200 -600>
    liftCoefficient=<0.1 0.3>
    linearVelocityScale=<0.2 0.15 0.15>
   }
   ziplining {
    angularVelocityScale=0.01
    fictitiousAngularScale=0
    jointName=root
    dragCoefficient=<0 0.05>
    gravityOverride=<0 0 -980>
    liftCoefficient=<0 0.05>
    linearVelocityScale=<0.1 0.1 0.1>
   }
  }
 }
 Cloth Paramaters_0 {
  type=ClothParamsStruct
  default {
   dBNO {
    angularVelocityScale=0.6
    fictitiousAngularScale=0.1
    jointName=pelvis
    dragCoefficient=<0.2 1>
    gravityOverride=<-980 400 0>
    liftCoefficient=<0 0.5>
    linearVelocityScale=<0.3 0.3 0.1>
   }
   dBNO_Carried {
    angularVelocityScale=0.6
    fictitiousAngularScale=0.1
    jointName=pelvis
    dragCoefficient=<0.2 1>
    gravityOverride=<-980 -200 0>
    liftCoefficient=<0 0.5>
    linearVelocityScale=<0.3 0.3 0.1>
   }
   dBNO_Impulse {
    angularVelocityScale=0
    fictitiousAngularScale=0
    jointName=pelvis
    dragCoefficient=<0 0.2>
    gravityOverride=<-3000 -2000 0>
    liftCoefficient=<0 0.1>
    linearVelocityScale=<0.3 0.1 0.3>
   }
   dBNO_RollBwd {
    angularVelocityScale=0
    fictitiousAngularScale=0
    jointName=pelvis
    dragCoefficient=<0 0.2>
    gravityOverride=<-2000 0 0>
    liftCoefficient=<0 0.1>
    linearVelocityScale=<0.3 0.1 0.3>
   }
   dBNO_RollFwd {
    angularVelocityScale=0
    fictitiousAngularScale=0
    jointName=pelvis
    dragCoefficient=<0 0.2>
    gravityOverride=<-980 0 0>
    liftCoefficient=<0 0.1>
    linearVelocityScale=<0.3 0.1 0.3>
   }
   dBNO_Tumbling {
    angularVelocityScale=0
    fictitiousAngularScale=0
    jointName=pelvis
    dragCoefficient=<0 0.2>
    gravityOverride=<-2500 -300 0>
    liftCoefficient=<0 0.1>
    linearVelocityScale=<0.3 0.1 0.3>
   }
   emoteOrMelee {
    angularVelocityScale=0.6
    fictitiousAngularScale=0.1
    jointName=root
    dragCoefficient=<2.5 2.5>
    gravityOverride=<0 0 -980>
    liftCoefficient=<0.5 0.5>
    linearVelocityScale=<0.6 0.6 0.5>
   }
   front_End {
    angularVelocityScale=0.05
    fictitiousAngularScale=0.6
    jointName=root
    dragCoefficient=<0.2 0.2>
    gravityOverride=<0 -200 -980>
    liftCoefficient=<0.1 0.1>
    linearVelocityScale=<0.1 0.1 0.1>
   }
   onGround_Crouch_Moving {
    angularVelocityScale=0.1
    fictitiousAngularScale=0.3
    jointName=root
    dragCoefficient=<2 2>
    gravityOverride=<0 0 -980>
    liftCoefficient=<0.5 0.5>
    linearVelocityScale=<0.6 0.6 0.5>
   }
   onGround_Falling {
    angularVelocityScale=0.6
    fictitiousAngularScale=0
    jointName=root
    dragCoefficient=<0 0.2>
    gravityOverride=<0 -500 -980>
    liftCoefficient=<0 0.1>
    linearVelocityScale=<2 1 0.3>
   }
   onGround_Moving {
    angularVelocityScale=0.6
    fictitiousAngularScale=0.1
    jointName=root
    dragCoefficient=<2 2>
    gravityOverride=<0 0 -980>
    liftCoefficient=<0.5 0.5>
    linearVelocityScale=<0.6 0.6 0.5>
   }
   onGround_Standing {
    angularVelocityScale=0.05
    fictitiousAngularScale=1.2
    jointName=root
    dragCoefficient=<2.5 2.5>
    gravityOverride=<0 0 -980>
    liftCoefficient=<0.5 0.5>
    linearVelocityScale=<0.6 0.6 0.5>
   }
   rail_Grinding {
    angularVelocityScale=0.2
    fictitiousAngularScale=0
    jointName=root
    dragCoefficient=<0.1 0.2>
    gravityOverride=<50 -500 -400>
    liftCoefficient=<0 0.5>
    linearVelocityScale=<0.1 0.1 0.1>
   }
   riding_Moving {
    angularVelocityScale=0.05
    fictitiousAngularScale=1
    jointName=root
    dragCoefficient=<0.5 0.5>
    gravityOverride=<0 0 -980>
    liftCoefficient=<0.5 0.5>
    linearVelocityScale=<0.4 0.6 0.3>
   }
   riding_Moving_Targeting {
    angularVelocityScale=0.05
    fictitiousAngularScale=0.7
    jointName=root
    dragCoefficient=<0.5 0.5>
    gravityOverride=<0 0 -980>
    liftCoefficient=<0.5 0.5>
    linearVelocityScale=<0.2 0.6 0.2>
   }
   riding_Standing {
    angularVelocityScale=0.05
    fictitiousAngularScale=1
    jointName=root
    dragCoefficient=<0.5 0.5>
    gravityOverride=<0 -450 -980>
    liftCoefficient=<0.5 0.5>
    linearVelocityScale=<0.6 0.6 0.3>
   }
   skydive_DivingDown {
    angularVelocityScale=0.1
    fictitiousAngularScale=0
    jointName=pelvis
    dragCoefficient=<0.1 0.1>
    gravityOverride=<-450 -100 0>
    liftCoefficient=<0.05 0.05>
    linearVelocityScale=<0.1 0.1 0.1>
   }
   skydive_FloatingDown {
    angularVelocityScale=0.1
    fictitiousAngularScale=0
    jointName=root
    dragCoefficient=<0.1 0.1>
    gravityOverride=<0 -800 200>
    liftCoefficient=<0.05 0.05>
    linearVelocityScale=<0.2 0.2 0.1>
   }
   skydive_GoingUp {
    angularVelocityScale=0.1
    fictitiousAngularScale=0
    jointName=root
    dragCoefficient=<0.01 0.01>
    gravityOverride=<0 -2000 400>
    liftCoefficient=<0.05 0.05>
    linearVelocityScale=<0.1 0.1 0.01>
   }
   skydive_Parachuting {
    angularVelocityScale=0.6
    fictitiousAngularScale=0.1
    jointName=root
    dragCoefficient=<0.8 0.8>
    gravityOverride=<0 -600 -500>
    liftCoefficient=<0.8 0.8>
    linearVelocityScale=<0.2 0.2 0.3>
   }
   skydive_Surfing {
    angularVelocityScale=0.6
    fictitiousAngularScale=0.1
    jointName=root
    dragCoefficient=<1.5 1.5>
    gravityOverride=<0 -600 -500>
    liftCoefficient=<1 1>
    linearVelocityScale=<0.1 0.1 0.3>
   }
   swimming {
    angularVelocityScale=0.6
    fictitiousAngularScale=0
    jointName=root
    dragCoefficient=<2.5 2.5>
    gravityOverride=<0 0 0>
    liftCoefficient=<0.1 0.1>
    linearVelocityScale=<0.6 0.6 0.5>
   }
   tactical_Sprint {
    angularVelocityScale=0.6
    fictitiousAngularScale=0.1
    jointName=root
    dragCoefficient=<2 2>
    gravityOverride=<0 0 -980>
    liftCoefficient=<0.5 0.5>
    linearVelocityScale=<0.6 0.6 0.5>
   }
   ziplining {
    angularVelocityScale=0.6
    fictitiousAngularScale=0
    jointName=root
    dragCoefficient=<0 1.5>
    gravityOverride=<0 0 -980>
    liftCoefficient=<0 0.3>
    linearVelocityScale=<0.4 0.4 0.5>
   }
  }
 }
 CrouchMelee {type=boolean}
 DeltaTimeLocal {type=number}
 Gravity Override {type=Vector}
 Gravity Override Params {
  type=GravityOverrideParamsStruct
  default {
   emoteOrMelee {jointName=pelvis, gravityOverride=<-800 -100 0>}
   rail_Grinding {gravityOverride=<0 -200 -400>}
   skydive_DivingDown {gravityOverride=<-1500 -200 400>}
   skydive_FloatingDown {gravityOverride=<0 -200 0>, windAmplitude=<100 -250 100>, windFrequency=<1.5 3.5 1>, windOffset=<0 -300 200>}
   skydive_GoingUp {gravityOverride=<0 -50 1000>}
   skydive_Parachuting {gravityOverride=<0 0 -400>, windFrequency=<1.5 2 1>}
   swimming {jointName=root, gravityOverride=<0 -200 0>}
   tactical_Sprint {gravityOverride=<0 -600 -600>}
   ziplining {gravityOverride=<0 -200 -600>}
  }
 }
 Is Backbling A Cape {type=boolean}
 IsOnZipline {type=boolean}
 isWearingCape {type=boolean}
 isWearingPack {type=boolean}
 RBAN Params {
  type=RBANParamsStruct
  default {
   dBNO {
    componentLinearAccScale=<0.4 0.4 0.4>
    componentLinearVelScale=<0.4 0.4 0.4>
    simSpaceSettings {
     maxAngularAcceleration=500
     maxAngularVelocity=3
     maxLinearAcceleration=350
     maxLinearVelocity=250
     velocityScaleZ=0
     worldAlpha=0.5
     externalLinearDragV=<0.1 0.1 0>
    }
   }
   dBNO_Carried {
    componentLinearAccScale=<0.4 0.4 0.4>
    componentLinearVelScale=<0.4 0.4 0.4>
    simSpaceSettings {
     maxAngularAcceleration=500
     maxAngularVelocity=3
     maxLinearAcceleration=350
     maxLinearVelocity=250
     velocityScaleZ=0
     worldAlpha=0.5
     externalLinearDragV=<0.1 0.1 0>
    }
   }
   emoteOrMelee {
    componentLinearAccScale=<0.2 0.2 0.2>
    componentLinearVelScale=<0.3 0.3 0.3>
    simSpaceSettings {
     maxAngularAcceleration=0.01
     maxAngularVelocity=0.02
     maxLinearAcceleration=350
     maxLinearVelocity=250
     velocityScaleZ=0
     worldAlpha=0.5
     externalLinearDragV=<0 0 0>
    }
   }
   front_End {simSpaceSettings {
     maxAngularAcceleration=50
     maxAngularVelocity=3
     maxLinearAcceleration=50
     maxLinearVelocity=50
     velocityScaleZ=0
     worldAlpha=0.5
     externalLinearDragV=<0 0 0>
    }}
   onGround_Crouch_Moving {
    componentLinearAccScale=<0.4 0.4 0.4>
    componentLinearVelScale=<0.4 0.4 0.4>
    simSpaceSettings {
     maxAngularAcceleration=500
     maxAngularVelocity=3
     maxLinearAcceleration=350
     maxLinearVelocity=250
     velocityScaleZ=0
     worldAlpha=0.5
     externalLinearDragV=<0.1 0.1 0>
    }
   }
   onGround_Falling {
    componentLinearAccScale=<0.4 0.4 0.4>
    componentLinearVelScale=<0.4 0.4 0.4>
    simSpaceSettings {
     maxAngularAcceleration=500
     maxAngularVelocity=3
     maxLinearAcceleration=350
     maxLinearVelocity=250
     velocityScaleZ=0
     worldAlpha=0.5
     externalLinearDragV=<0.1 0.1 0>
    }
   }
   onGround_Moving {
    componentLinearAccScale=<0.6 0.4 0.9>
    componentLinearVelScale=<0.7 0.5 1>
    simSpaceSettings {
     maxAngularAcceleration=1
     maxAngularVelocity=5
     maxLinearAcceleration=300
     maxLinearVelocity=500
     velocityScaleZ=1
     worldAlpha=0.5
     externalLinearDragV=<0 0 0>
    }
   }
   onGround_Standing {
    componentLinearAccScale=<0.5 0.5 0.5>
    componentLinearVelScale=<0.6 0.6 0.6>
    simSpaceSettings {
     maxAngularAcceleration=5
     maxAngularVelocity=10
     maxLinearAcceleration=150
     maxLinearVelocity=300
     velocityScaleZ=1
     worldAlpha=0.5
     externalLinearDragV=<0 0 0>
    }
   }
   rail_Grinding {
    componentLinearAccScale=<0.4 0.4 0.4>
    componentLinearVelScale=<0.4 0.4 0.4>
    simSpaceSettings {
     maxAngularAcceleration=500
     maxAngularVelocity=3
     maxLinearAcceleration=350
     maxLinearVelocity=250
     velocityScaleZ=0
     worldAlpha=0.5
     externalLinearDragV=<0.1 0.1 0>
    }
   }
   riding_Moving {
    componentLinearAccScale=<0.4 0.4 0.4>
    componentLinearVelScale=<0.4 0.4 0.4>
    simSpaceSettings {
     maxAngularAcceleration=500
     maxAngularVelocity=3
     maxLinearAcceleration=350
     maxLinearVelocity=250
     velocityScaleZ=0
     worldAlpha=0.5
     externalLinearDragV=<0.1 0.1 0>
    }
   }
   riding_Moving_Targeting {
    componentLinearAccScale=<0.4 0.4 0.4>
    componentLinearVelScale=<0.4 0.4 0.4>
    simSpaceSettings {
     maxAngularAcceleration=500
     maxAngularVelocity=3
     maxLinearAcceleration=350
     maxLinearVelocity=250
     velocityScaleZ=0
     worldAlpha=0.5
     externalLinearDragV=<0.1 0.1 0>
    }
   }
   riding_Standing {
    componentLinearAccScale=<0.4 0.4 0.4>
    componentLinearVelScale=<0.4 0.4 0.4>
    simSpaceSettings {
     maxAngularAcceleration=500
     maxAngularVelocity=3
     maxLinearAcceleration=350
     maxLinearVelocity=250
     velocityScaleZ=0
     worldAlpha=0.5
     externalLinearDragV=<0.1 0.1 0>
    }
   }
   skydive_DivingDown {
    componentLinearAccScale=<0.3 0.3 0.3>
    componentLinearVelScale=<0.4 0.4 0.4>
    simSpaceSettings {
     maxAngularAcceleration=2
     maxAngularVelocity=5
     maxLinearAcceleration=200
     maxLinearVelocity=300
     velocityScaleZ=0
     worldAlpha=0.5
     externalLinearDragV=<0 0 0>
    }
   }
   skydive_FloatingDown {
    componentLinearAccScale=<0.1 0.2 0.1>
    componentLinearVelScale=<0.2 0.3 0.2>
    simSpaceSettings {
     maxAngularAcceleration=1
     maxAngularVelocity=2
     maxLinearAcceleration=150
     maxLinearVelocity=300
     velocityScaleZ=0
     worldAlpha=0.5
     externalLinearDragV=<0 0 0>
    }
   }
   skydive_GoingUp {
    componentLinearAccScale=<0.3 0.3 0.3>
    componentLinearVelScale=<0.4 0.4 0.4>
    simSpaceSettings {
     maxAngularAcceleration=2
     maxAngularVelocity=5
     maxLinearAcceleration=200
     maxLinearVelocity=300
     velocityScaleZ=0
     worldAlpha=0.5
     externalLinearDragV=<0.1 0.1 0>
    }
   }
   skydive_Parachuting {
    componentLinearAccScale=<0.3 0.3 0.3>
    componentLinearVelScale=<0.4 0.4 0.4>
    simSpaceSettings {
     maxAngularAcceleration=2
     maxAngularVelocity=5
     maxLinearAcceleration=200
     maxLinearVelocity=300
     velocityScaleZ=0
     worldAlpha=0.5
     externalLinearDragV=<0 0 0>
    }
   }
   skydive_Surfing {
    componentLinearAccScale=<0.4 0.4 0.4>
    componentLinearVelScale=<0.4 0.4 0.4>
    simSpaceSettings {
     maxAngularAcceleration=500
     maxAngularVelocity=3
     maxLinearAcceleration=350
     maxLinearVelocity=250
     velocityScaleZ=0
     worldAlpha=0.5
     externalLinearDragV=<0.1 0.1 0>
    }
   }
   swimming {
    componentLinearAccScale=<0.2 0.2 0.2>
    componentLinearVelScale=<0.3 0.3 0.3>
    simSpaceSettings {
     maxAngularAcceleration=1
     maxAngularVelocity=3
     maxLinearAcceleration=350
     maxLinearVelocity=250
     velocityScaleZ=0
     worldAlpha=0.5
     externalLinearDragV=<0.1 0.1 0>
    }
   }
   tactical_Sprint {
    componentLinearAccScale=<0.05 0.2 0.2>
    componentLinearVelScale=<0.1 0.3 0.3>
    simSpaceSettings {
     maxAngularAcceleration=0.1
     maxAngularVelocity=1
     maxLinearAcceleration=12
     maxLinearVelocity=24
     velocityScaleZ=0
     worldAlpha=0.5
     externalLinearDragV=<0 0 0>
    }
   }
   ziplining {
    componentLinearAccScale=<0.4 0.4 0.4>
    componentLinearVelScale=<0.4 0.4 0.4>
    simSpaceSettings {
     maxAngularAcceleration=500
     maxAngularVelocity=3
     maxLinearAcceleration=350
     maxLinearVelocity=250
     velocityScaleZ=0
     worldAlpha=0.5
     externalLinearDragV=<0.1 0.1 0>
    }
   }
  }
 }
 RBAN Settings Out {
  type=RBANSettingStruct
  default {
   componentLinearAccScale=<0 0 0>
   componentLinearVelScale=<0 0 0>
   simSpaceSettings {
    dampingAlpha=1
    maxAngularAcceleration=1000
    maxAngularVelocity=3
    maxLinearAcceleration=300
    maxLinearVelocity=700
    velocityScaleZ=0
    worldAlpha=1
    externalAngularVelocity=<0 0 0>
    externalLinearDragV=<0 0.5 0>
    externalLinearVelocity=<0 0 0>
   }
  }
 }
 resetClothLandingFromZilpine {type=boolean}
 Simulation Space {type=AnimPhysSimSpaceType}
 Wind Gust Local Settings {type=WindGustLocalSettingStruct, default {elapsedTime=0, finalGustStrengthLocal=0, gustDuration=0, gustStrength=0, timeBetweenGusts=0}}
 Wind Gusts Params {
  type=WindGustParamsStruct
  default {
   dBNO {
    maxDownInterpSpeed=5
    maxGustDuration=0
    maxGustStrength=0
    maxTimeBetweenGusts=7
    maxUpInterpSpeed=5
    minDownInterpSpeed=2
    minGustDuration=0
    minGustStrength=0
    minTimeBetweenGusts=0
    minUpInterpSpeed=2
    windTurbulenceScalarA=0
    windTurbulenceScalarB=0
    windAmplitude=<0 0 0>
    windFrequency=<0 0 0>
    windOffset=<0 0 0>
   }
   dBNO_Carried {
    maxDownInterpSpeed=5
    maxGustDuration=0
    maxGustStrength=0
    maxTimeBetweenGusts=7
    maxUpInterpSpeed=5
    minDownInterpSpeed=2
    minGustDuration=0
    minGustStrength=0
    minTimeBetweenGusts=0
    minUpInterpSpeed=2
    windTurbulenceScalarA=0
    windTurbulenceScalarB=0
    windAmplitude=<0 0 0>
    windFrequency=<0 0 0>
    windOffset=<0 0 0>
   }
   dBNO_Impulse {
    maxDownInterpSpeed=5
    maxGustDuration=2
    maxGustStrength=7
    maxTimeBetweenGusts=7
    maxUpInterpSpeed=5
    minDownInterpSpeed=2
    minGustDuration=0.5
    minGustStrength=3
    minTimeBetweenGusts=0
    minUpInterpSpeed=2
    windTurbulenceScalarA=0.1
    windTurbulenceScalarB=1
    windAmplitude=<0 0 0>
    windFrequency=<0 0 0>
    windOffset=<0 0 0>
   }
   dBNO_RollBwd {
    maxDownInterpSpeed=5
    maxGustDuration=2
    maxGustStrength=7
    maxTimeBetweenGusts=7
    maxUpInterpSpeed=5
    minDownInterpSpeed=2
    minGustDuration=0.5
    minGustStrength=3
    minTimeBetweenGusts=0
    minUpInterpSpeed=2
    windTurbulenceScalarA=0.1
    windTurbulenceScalarB=1
    windAmplitude=<0 0 0>
    windFrequency=<0 0 0>
    windOffset=<0 0 0>
   }
   dBNO_RollFwd {
    maxDownInterpSpeed=5
    maxGustDuration=2
    maxGustStrength=7
    maxTimeBetweenGusts=7
    maxUpInterpSpeed=5
    minDownInterpSpeed=2
    minGustDuration=0.5
    minGustStrength=3
    minTimeBetweenGusts=0
    minUpInterpSpeed=2
    windTurbulenceScalarA=0.1
    windTurbulenceScalarB=1
    windAmplitude=<0 0 0>
    windFrequency=<0 0 0>
    windOffset=<0 0 0>
   }
   dBNO_Tumbling {
    maxDownInterpSpeed=5
    maxGustDuration=2
    maxGustStrength=7
    maxTimeBetweenGusts=7
    maxUpInterpSpeed=5
    minDownInterpSpeed=2
    minGustDuration=0.5
    minGustStrength=3
    minTimeBetweenGusts=0
    minUpInterpSpeed=2
    windTurbulenceScalarA=0.1
    windTurbulenceScalarB=1
    windAmplitude=<0 0 0>
    windFrequency=<0 0 0>
    windOffset=<0 0 0>
   }
   emoteOrMelee {
    maxDownInterpSpeed=5
    maxGustDuration=0
    maxGustStrength=0
    maxTimeBetweenGusts=7
    maxUpInterpSpeed=5
    minDownInterpSpeed=2
    minGustDuration=0
    minGustStrength=0
    minTimeBetweenGusts=0
    minUpInterpSpeed=2
    windTurbulenceScalarA=0
    windTurbulenceScalarB=0
    windAmplitude=<0 0 0>
    windFrequency=<0 0 0>
    windOffset=<0 0 0>
   }
   front_End {
    maxDownInterpSpeed=5
    maxGustDuration=0
    maxGustStrength=0
    maxTimeBetweenGusts=7
    maxUpInterpSpeed=5
    minDownInterpSpeed=2
    minGustDuration=0
    minGustStrength=0
    minTimeBetweenGusts=0
    minUpInterpSpeed=2
    windTurbulenceScalarA=0
    windTurbulenceScalarB=0
    windAmplitude=<0 0 0>
    windFrequency=<0 0 0>
    windOffset=<0 0 0>
   }
   onGround_Crouch_Moving {
    maxDownInterpSpeed=5
    maxGustDuration=0
    maxGustStrength=0
    maxTimeBetweenGusts=7
    maxUpInterpSpeed=5
    minDownInterpSpeed=2
    minGustDuration=0
    minGustStrength=0
    minTimeBetweenGusts=0
    minUpInterpSpeed=2
    windTurbulenceScalarA=0
    windTurbulenceScalarB=0
    windAmplitude=<0 0 0>
    windFrequency=<0 0 0>
    windOffset=<0 0 0>
   }
   onGround_Falling {
    maxDownInterpSpeed=5
    maxGustDuration=0
    maxGustStrength=0
    maxTimeBetweenGusts=7
    maxUpInterpSpeed=5
    minDownInterpSpeed=2
    minGustDuration=0
    minGustStrength=0
    minTimeBetweenGusts=0
    minUpInterpSpeed=2
    windTurbulenceScalarA=0
    windTurbulenceScalarB=0
    windAmplitude=<0 0 0>
    windFrequency=<0 0 0>
    windOffset=<0 0 0>
   }
   onGround_Moving {
    maxDownInterpSpeed=5
    maxGustDuration=8
    maxGustStrength=7
    maxTimeBetweenGusts=8
    maxUpInterpSpeed=5
    minDownInterpSpeed=2
    minGustDuration=0.2
    minGustStrength=3
    minTimeBetweenGusts=1
    minUpInterpSpeed=2
    windTurbulenceScalarA=0.1
    windTurbulenceScalarB=0.4
    windAmplitude=<30 35 42>
    windFrequency=<1.1 1 1.4>
    windOffset=<0 -200 0>
   }
   onGround_Standing {
    maxDownInterpSpeed=5
    maxGustDuration=8
    maxGustStrength=7
    maxTimeBetweenGusts=8
    maxUpInterpSpeed=5
    minDownInterpSpeed=2
    minGustDuration=0.2
    minGustStrength=2
    minTimeBetweenGusts=1
    minUpInterpSpeed=2
    windTurbulenceScalarA=0
    windTurbulenceScalarB=0
    windAmplitude=<50 24 28>
    windFrequency=<4 3 5>
    windOffset=<50 -100 -150>
   }
   rail_Grinding {
    maxDownInterpSpeed=5
    maxGustDuration=8
    maxGustStrength=7
    maxTimeBetweenGusts=8
    maxUpInterpSpeed=5
    minDownInterpSpeed=2
    minGustDuration=0.2
    minGustStrength=3
    minTimeBetweenGusts=1
    minUpInterpSpeed=2
    windTurbulenceScalarA=0.1
    windTurbulenceScalarB=4
    windAmplitude=<0 25 35>
    windFrequency=<0 4 10>
    windOffset=<0 -500 -300>
   }
   riding_Moving {
    maxDownInterpSpeed=5
    maxGustDuration=2
    maxGustStrength=7
    maxTimeBetweenGusts=7
    maxUpInterpSpeed=5
    minDownInterpSpeed=2
    minGustDuration=0.5
    minGustStrength=3
    minTimeBetweenGusts=0
    minUpInterpSpeed=2
    windTurbulenceScalarA=0.1
    windTurbulenceScalarB=1
    windAmplitude=<0 0 0>
    windFrequency=<0 0 0>
    windOffset=<0 0 0>
   }
   riding_Moving_Targeting {
    maxDownInterpSpeed=5
    maxGustDuration=2
    maxGustStrength=7
    maxTimeBetweenGusts=7
    maxUpInterpSpeed=5
    minDownInterpSpeed=2
    minGustDuration=0.5
    minGustStrength=3
    minTimeBetweenGusts=0
    minUpInterpSpeed=2
    windTurbulenceScalarA=0.1
    windTurbulenceScalarB=1
    windAmplitude=<0 0 0>
    windFrequency=<0 0 0>
    windOffset=<0 0 0>
   }
   riding_Standing {
    maxDownInterpSpeed=5
    maxGustDuration=2
    maxGustStrength=7
    maxTimeBetweenGusts=7
    maxUpInterpSpeed=5
    minDownInterpSpeed=2
    minGustDuration=0.5
    minGustStrength=3
    minTimeBetweenGusts=0
    minUpInterpSpeed=2
    windTurbulenceScalarA=0.1
    windTurbulenceScalarB=1
    windAmplitude=<0 0 0>
    windFrequency=<0 0 0>
    windOffset=<0 0 0>
   }
   skydive_DivingDown {
    maxDownInterpSpeed=5
    maxGustDuration=8
    maxGustStrength=7
    maxTimeBetweenGusts=8
    maxUpInterpSpeed=5
    minDownInterpSpeed=2
    minGustDuration=0.2
    minGustStrength=3
    minTimeBetweenGusts=1
    minUpInterpSpeed=2
    windTurbulenceScalarA=0.1
    windTurbulenceScalarB=4
    windAmplitude=<0 25 35>
    windFrequency=<0 4 10>
    windOffset=<-450 -100 100>
   }
   skydive_FloatingDown {
    maxDownInterpSpeed=2
    maxGustDuration=4
    maxGustStrength=2
    maxTimeBetweenGusts=4
    maxUpInterpSpeed=8
    minDownInterpSpeed=0.5
    minGustDuration=0.2
    minGustStrength=1
    minTimeBetweenGusts=1
    minUpInterpSpeed=4
    windTurbulenceScalarA=4
    windTurbulenceScalarB=5
    windAmplitude=<45 45 45>
    windFrequency=<4 6 4>
    windOffset=<-150 -400 200>
   }
   skydive_GoingUp {
    maxDownInterpSpeed=5
    maxGustDuration=8
    maxGustStrength=2
    maxTimeBetweenGusts=8
    maxUpInterpSpeed=5
    minDownInterpSpeed=2
    minGustDuration=0.2
    minGustStrength=1
    minTimeBetweenGusts=1
    minUpInterpSpeed=2
    windTurbulenceScalarA=1
    windTurbulenceScalarB=3
    windAmplitude=<45 45 45>
    windFrequency=<4 6 4>
    windOffset=<-150 -400 200>
   }
   skydive_Parachuting {
    maxDownInterpSpeed=5
    maxGustDuration=8
    maxGustStrength=3
    maxTimeBetweenGusts=8
    maxUpInterpSpeed=2
    minDownInterpSpeed=2
    minGustDuration=5
    minGustStrength=1
    minTimeBetweenGusts=5
    minUpInterpSpeed=0.5
    windTurbulenceScalarA=0.5
    windTurbulenceScalarB=1.4
    windAmplitude=<25 50 75>
    windFrequency=<8 6 4>
    windOffset=<-150 -400 -150>
   }
   skydive_Surfing {
    maxDownInterpSpeed=5
    maxGustDuration=2
    maxGustStrength=7
    maxTimeBetweenGusts=7
    maxUpInterpSpeed=5
    minDownInterpSpeed=2
    minGustDuration=0.5
    minGustStrength=3
    minTimeBetweenGusts=0
    minUpInterpSpeed=2
    windTurbulenceScalarA=0.1
    windTurbulenceScalarB=1
    windAmplitude=<40 40 40>
    windFrequency=<1.1 1.2 1.7>
    windOffset=<0 0 -300>
   }
   swimming {
    maxDownInterpSpeed=5
    maxGustDuration=0
    maxGustStrength=0
    maxTimeBetweenGusts=7
    maxUpInterpSpeed=5
    minDownInterpSpeed=2
    minGustDuration=0
    minGustStrength=0
    minTimeBetweenGusts=0
    minUpInterpSpeed=2
    windTurbulenceScalarA=0
    windTurbulenceScalarB=0
    windAmplitude=<0 0 0>
    windFrequency=<0 0 0>
    windOffset=<0 0 0>
   }
   tactical_Sprint {
    maxDownInterpSpeed=5
    maxGustDuration=2
    maxGustStrength=7
    maxTimeBetweenGusts=7
    maxUpInterpSpeed=5
    minDownInterpSpeed=2
    minGustDuration=0.5
    minGustStrength=3
    minTimeBetweenGusts=0
    minUpInterpSpeed=2
    windTurbulenceScalarA=0.1
    windTurbulenceScalarB=1
    windAmplitude=<12 24 20>
    windFrequency=<3 4 6>
    windOffset=<0 -100 -300>
   }
   ziplining {
    maxDownInterpSpeed=5
    maxGustDuration=8
    maxGustStrength=7
    maxTimeBetweenGusts=8
    maxUpInterpSpeed=5
    minDownInterpSpeed=2
    minGustDuration=0.2
    minGustStrength=3
    minTimeBetweenGusts=1
    minUpInterpSpeed=2
    windTurbulenceScalarA=0.1
    windTurbulenceScalarB=4
    windAmplitude=<0 25 35>
    windFrequency=<0 4 10>
    windOffset=<0 -500 -300>
   }
  }
 }
]
