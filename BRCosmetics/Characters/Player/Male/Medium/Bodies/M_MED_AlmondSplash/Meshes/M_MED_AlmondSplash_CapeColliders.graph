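# Physics asset of collision shapes for the medium male player body AlmondSplash, named as the colliders for its cape.
# Capsules cover the pelvis, spine, clavicles and neck, and a sphere covers the head, while arm, leg and foot bodies have no shapes.
# Offset capsules also sit on the dyn_jacket bone chains, so the coat's front, side, middle and back act as colliders.
# It defines no constraints.
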PhysicsAsset /BRCosmetics/Characters/Player/Male/Medium/Bodies/M_MED_AlmondSplash/Meshes/M_MED_AlmondSplash_CapeColliders
bodies {
 calf_l []
 calf_r []
 clavicle_l [{shapeType=Capsule, length=9.0192, radius=9.8871, center=<9.4266 -1.1806 -2.4977>, rotation=<rot 69.3297 1.4476 176.9462>}]
 clavicle_r [{shapeType=Capsule, length=9.0192, radius=9.8871, center=<-9.4266 1.1806 2.4977>, rotation=<rot 69.3297 1.4476 -3.0538>}]
 dyn_jacket_bk_03 [{shapeType=Capsule, length=10.8333, radius=9.4775, shapeName=dyn_coat_fr_04_L_capsule, center=<8.5855 0 -6.2647>, rotation=<rot 0 0 90>}]
 dyn_jacket_bk_l_01 [{shapeType=Capsule, length=12.1487, radius=8.1621, shapeName=dyn_coat_fr_04_L_capsule, center=<5 0 -5.8274>, rotation=<rot 0 0 90>}]
 dyn_jacket_bk_l_02 [{shapeType=Capsule, length=12.1487, radius=8.1621, shapeName=dyn_coat_fr_04_L_capsule, center=<5 0 -5.9088>, rotation=<rot 0 0 90>}]
 dyn_jacket_bk_r_01 [{shapeType=Capsule, length=11.0379, radius=9.2729, shapeName=dyn_coat_fr_04_L_capsule, center=<5 0 -5.8274>, rotation=<rot 0 0 90>}]
 dyn_jacket_bk_r_02 [{shapeType=Capsule, length=11.0379, radius=9.2729, shapeName=dyn_coat_fr_04_L_capsule, center=<5 0 -5.9088>, rotation=<rot 0 0 90>}]
 dyn_jacket_fr_l_04 [{shapeType=Capsule, length=17.6037, radius=7, shapeName=dyn_coat_fr_04_L_capsule, center=<5 0 -6.2647>, rotation=<rot 0 0 90>}]
 dyn_jacket_fr_l_05 [{shapeType=Capsule, length=17.6037, radius=7, shapeName=dyn_coat_fr_04_L_capsule, center=<5 0 -6.2647>, rotation=<rot 0 0 90>}]
 dyn_jacket_fr_l_06 [{shapeType=Capsule, length=17.6037, radius=7, shapeName=dyn_coat_fr_04_L_capsule, center=<5 0 -6.2647>, rotation=<rot 0 0 90>}]
 dyn_jacket_fr_r_04 [{shapeType=Capsule, length=17.6037, radius=7, shapeName=dyn_coat_fr_04_L_capsule, center=<5 0 -6.2647>, rotation=<rot 0 0 90>}]
 dyn_jacket_fr_r_05 [{shapeType=Capsule, length=17.6037, radius=7, shapeName=dyn_coat_fr_04_L_capsule, center=<5 0 -6.2647>, rotation=<rot 0 0 90>}]
 dyn_jacket_fr_r_06 [{shapeType=Capsule, length=17.6037, radius=7, shapeName=dyn_coat_fr_04_L_capsule, center=<5 0 -6.2647>, rotation=<rot 0 0 90>}]
 dyn_jacket_l_03 [{shapeType=Capsule, length=13.3108, radius=7, shapeName=dyn_coat_fr_04_L_capsule, center=<5 0 -6.2647>, rotation=<rot 0 0 90>}]
 dyn_jacket_l_04 [{shapeType=Capsule, length=13.3108, radius=7, shapeName=dyn_coat_fr_04_L_capsule, center=<5 0 -6.2647>, rotation=<rot 0 0 90>}]
 dyn_jacket_l_05 [{shapeType=Capsule, length=13.3108, radius=7, shapeName=dyn_coat_fr_04_L_capsule, center=<5 0 -6.2647>, rotation=<rot 0 0 90>}]
 dyn_jacket_mid_l_03 [{shapeType=Capsule, length=11.0379, radius=9.2729, shapeName=dyn_coat_fr_04_L_capsule, center=<5 0 -6.2647>, rotation=<rot 0 0 90>}]
 dyn_jacket_mid_l_04 [{shapeType=Capsule, length=11.0379, radius=9.2729, shapeName=dyn_coat_fr_04_L_capsule, center=<5 0 -5.4184>, rotation=<rot 0 0 90>}]
 dyn_jacket_mid_l_05 [{shapeType=Capsule, length=11.0379, radius=9.2729, shapeName=dyn_coat_fr_04_L_capsule, center=<5 0 -6.2248>, rotation=<rot 0 0 90>}]
 dyn_jacket_mid_r_03 [{shapeType=Capsule, length=11.0379, radius=9.2729, shapeName=dyn_coat_fr_04_L_capsule, center=<5 0 -6.2647>, rotation=<rot 0 0 90>}]
 dyn_jacket_mid_r_04 [{shapeType=Capsule, length=11.0379, radius=9.2729, shapeName=dyn_coat_fr_04_L_capsule, center=<5 0 -5.4184>, rotation=<rot 0 0 90>}]
 dyn_jacket_mid_r_05 [{shapeType=Capsule, length=11.0379, radius=9.2729, shapeName=dyn_coat_fr_04_L_capsule, center=<5 0 -6.2248>, rotation=<rot 0 0 90>}]
 dyn_jacket_r_03 [{shapeType=Capsule, length=13.3108, radius=7, shapeName=dyn_coat_fr_04_L_capsule, center=<5 0 -6.2647>, rotation=<rot 0 0 90>}]
 dyn_jacket_r_04 [{shapeType=Capsule, length=13.3108, radius=7, shapeName=dyn_coat_fr_04_L_capsule, center=<5 0 -6.2647>, rotation=<rot 0 0 90>}]
 dyn_jacket_r_05 [{shapeType=Capsule, length=13.3108, radius=7, shapeName=dyn_coat_fr_04_L_capsule, center=<5 0 -6.2647>, rotation=<rot 0 0 90>}]
 foot_l []
 foot_r []
 head [{shapeType=Sphere, radius=18.2413, center=<9.2123 -2.2909 0>}]
 lowerarm_l []
 lowerarm_r []
 neck_02 [{shapeType=Capsule, length=13.4375, radius=7.3175, center=<-0.2642 -0.3408 0>, rotation=<rot 90 -171.1193 -175.3978>}]
 pelvis [{shapeType=Capsule, length=24.5919, radius=16.4861, center=<-2.4887 0.1881 0>}]
 spine_01 [{shapeType=Capsule, length=25.4313, radius=16.0094, center=<-1.7965 0.1884 0>}]
 spine_02 [
  {shapeType=Capsule, length=25.0655, radius=13.6322, center=<-1.711 -0.1851 0.0001>}
  {shapeType=Capsule, length=29.3956, radius=11.5807, center=<2.5024 0.4992 3.9825>, rotation=<rot 90 0.0049 0.0049>}
  {shapeType=Capsule, length=29.875, radius=11.1013, center=<2.4704 0.4991 -3.9825>, rotation=<rot 90 -0.0396 -0.0396>}
 ]
 spine_03 [{shapeType=Capsule, length=22.6162, radius=14.2578, center=<0.1445 0.2396 0.0001>}]
 spine_04 [{shapeType=Capsule, length=24.3765, radius=14.7832, center=<2.766 0.7765 0.0002>}]
 spine_05 [
  {shapeType=Capsule, length=5.976, radius=17.7416, center=<9.4958 0.4385 0>}
  {shapeType=Capsule, length=13.9518, radius=13.9425, center=<-0.2916 0.4452 0>}
 ]
 thigh_l []
 thigh_r []
 upperarm_l []
 upperarm_r []
}
constraints []
