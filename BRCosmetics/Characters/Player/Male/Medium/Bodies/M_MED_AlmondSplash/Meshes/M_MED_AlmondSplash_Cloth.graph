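# Cloth collision physics asset for the M_MED_AlmondSplash male medium player body, judging by its name.
# Capsules cover the pelvis, three spine bones and feet, spheres cover the thighs, and both calf bodies list no shapes.
# A huge sphere on root, offset far downward so its top sits just above the origin, approximates a ground surface.
# It has no constraints, so no joints link the bodies.
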PhysicsAsset /BRCosmetics/Characters/Player/Male/Medium/Bodies/M_MED_AlmondSplash/Meshes/M_MED_AlmondSplash_Cloth
bodies {
 calf_l []
 calf_r []
 foot_l [{shapeType=Capsule, length=13.7232, radius=8.6453, shapeName=foot_l_capsule, center=<-3.9254 5.8957 0.0583>, rotation=<rot 0 6.6027 -79.4871>}]
 foot_r [{
   shapeType=Capsule
   length=13.7232
   radius=8.6453
   shapeName=foot_l_capsule
   center=<3.9254 -5.8957 -0.0583>
   rotation=<rot 0 6.6027 100.5129>
  }]
 pelvis [{shapeType=Capsule, length=7.8998, radius=11.0377, shapeName=pelvis_capsule, center=<-4.0244 -0.163 0>}]
 root [{shapeType=Sphere, radius=300, shapeName=root_sphere, center=<0 0 -289.4513>}]
 spine_01 [{shapeType=Capsule, length=7.5532, radius=8.968, shapeName=spine_01_capsule, center=<0 0.4579 0.2019>}]
 spine_02 [{shapeType=Capsule, length=5.709, radius=8.5945, shapeName=spine_02_capsule, center=<-1.2125 0.4657 0.484>}]
 spine_03 [{shapeType=Capsule, length=8.4921, radius=8.3271, shapeName=spine_03_capsule}]
 thigh_l [{shapeType=Sphere, radius=8.1875, shapeName=thigh_l_sphere, center=<-15.4845 2.2403 0.767>}]
 thigh_r [{shapeType=Sphere, radius=8.1875, shapeName=thigh_l_sphere, center=<15.4845 -2.2403 -0.767>}]
}
constraints []
