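# Physics asset for the AlmondSplash male medium character's face accessory mesh, covering left and right braids and a hat chain.
# Each chain links capsule bodies from an MSTR root through swing- and twist-limited constraints, five segments per braid and four for the hat.
# Braid swing generally loosens toward the tips, while twist stays tight throughout.
# Unconstrained head, neck, upper spine and upper-arm capsules supply body shapes around the chains.
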PhysicsAsset /BRCosmetics/Characters/Player/Male/Medium/Bodies/M_MED_AlmondSplash/Meshes/Parts/M_MED_AlmondSplash_FaceAcc_Physics
bodies {
 dyn_braid_l_01 [{shapeType=Capsule, length=4.1622, radius=1.8378, shapeName=dyn_braid_l_01_capsule, center=<3.4785 0 0>, rotation=<rot 90 0 0>}]
 dyn_braid_l_02 [{shapeType=Capsule, length=4.4797, radius=1.5203, shapeName=dyn_braid_l_02_capsule, center=<3.4785 0 0>, rotation=<rot 90 0 0>}]
 dyn_braid_l_03 [{shapeType=Capsule, length=4.5887, radius=1.4113, shapeName=dyn_braid_l_03_capsule, center=<3.4785 0 0>, rotation=<rot 90 0 0>}]
 dyn_braid_l_04 [{shapeType=Capsule, length=4.7412, radius=1.2588, shapeName=dyn_braid_l_04_capsule, center=<3.4785 0 0>, rotation=<rot 86.6765 0.0004 0.0003>}]
 dyn_braid_l_05 [{shapeType=Capsule, length=1.1886, radius=3.1469, shapeName=dyn_braid_l_05_capsule, center=<2.282 0 0>, rotation=<rot 90 0 0>}]
 dyn_braid_l_MSTR [{shapeType=Capsule, length=4, radius=2, shapeName=dyn_braid_l_MSTR_capsule, rotation=<rot 90 0 0>}]
 dyn_braid_r_01 [{shapeType=Capsule, length=4.1622, radius=1.8378, shapeName=dyn_braid_r_01_capsule, center=<3.4785 0 0>, rotation=<rot 90 0 0>}]
 dyn_braid_r_02 [{shapeType=Capsule, length=4.4797, radius=1.5203, shapeName=dyn_braid_r_02_capsule, center=<3.4785 0 0>, rotation=<rot 90 0 0>}]
 dyn_braid_r_03 [{shapeType=Capsule, length=4.5887, radius=1.4113, shapeName=dyn_braid_r_03_capsule, center=<3.4785 0 0>, rotation=<rot 90 0 0>}]
 dyn_braid_r_04 [{
   shapeType=Capsule
   length=4.7412
   radius=1.2588
   shapeName=dyn_braid_r_04_capsule
   center=<3.4785 0 0>
   rotation=<rot 86.6765 -0.0005 -0.0003>
  }]
 dyn_braid_r_05 [{shapeType=Capsule, length=1.1886, radius=3.1469, shapeName=dyn_braid_r_05_capsule, center=<2.282 0 0>, rotation=<rot 90 0 0>}]
 dyn_braid_r_MSTR [{shapeType=Capsule, length=4, radius=2, shapeName=dyn_braid_r_MSTR_capsule, rotation=<rot 90 0 0>}]
 dyn_hat_01 [{shapeType=Capsule, length=4.0301, radius=1.9699, shapeName=dyn_hat_01_capsule, center=<2.1138 0 0>, rotation=<rot 90 0 0>}]
 dyn_hat_02 [{shapeType=Capsule, length=4.0456, radius=1.9544, shapeName=dyn_hat_02_capsule, center=<2.1138 0 0>, rotation=<rot 90 0 0>}]
 dyn_hat_03 [{shapeType=Capsule, length=3.8753, radius=2.1247, shapeName=dyn_hat_03_capsule, center=<2.1138 0 0>, rotation=<rot 90 0 0>}]
 dyn_hat_04 [{shapeType=Capsule, length=0.6942, radius=3.9626, shapeName=dyn_hat_04_capsule, center=<2.4069 0 -0.4015>, rotation=<rot 90 0 0>}]
 dyn_hat_MSTR [{shapeType=Capsule, length=4, radius=2, shapeName=dyn_hat_MSTR_capsule, rotation=<rot 90 0 0>}]
 head [{shapeType=Capsule, length=24.6957, radius=10.0618, shapeName=head_capsule, center=<-1.1088 1.0419 0>, rotation=<rot 90 0 0>}]
 neck_01 [{shapeType=Capsule, length=12.8179, radius=10.1868, shapeName=neck_01_capsule, center=<0.5304 -6.6208 0>}]
 neck_02 [{shapeType=Capsule, length=23.1129, radius=5.6232, shapeName=neck_02_capsule, center=<4.8515 0 0>, rotation=<rot 90 0 0>}]
 spine_05 [{shapeType=Capsule, length=45.2461, radius=12.2968, shapeName=spine_05_capsule, center=<4.9093 -0.303 0>}]
 upperarm_l [{shapeType=Capsule, length=35.3004, radius=8.8813, shapeName=upperarm_l_capsule, center=<17.897 0 2.0171>, rotation=<rot 90 0 0>}]
 upperarm_r [{shapeType=Capsule, length=35.3004, radius=8.8813, shapeName=upperarm_r_capsule, center=<-17.897 0 -2.0171>, rotation=<rot -90 180 180>}]
}
constraints [
 {bone1Name=dyn_braid_l_01, bone2Name=dyn_braid_l_MSTR, swing1LimitDegrees=30, swing2LimitDegrees=10, twistLimitDegrees=2}
 {bone1Name=dyn_braid_l_02, bone2Name=dyn_braid_l_01, swing1LimitDegrees=30, swing2LimitDegrees=15, twistLimitDegrees=2}
 {bone1Name=dyn_braid_l_03, bone2Name=dyn_braid_l_02, swing1LimitDegrees=40, swing2LimitDegrees=20, twistLimitDegrees=2}
 {bone1Name=dyn_braid_l_04, bone2Name=dyn_braid_l_03, swing1LimitDegrees=30, swing2LimitDegrees=30, twistLimitDegrees=2}
 {bone1Name=dyn_braid_l_05, bone2Name=dyn_braid_l_04, swing1LimitDegrees=60, swing2LimitDegrees=60, twistLimitDegrees=4}
 {bone1Name=dyn_braid_r_01, bone2Name=dyn_braid_r_MSTR, swing1LimitDegrees=30, swing2LimitDegrees=10, twistLimitDegrees=2}
 {bone1Name=dyn_braid_r_02, bone2Name=dyn_braid_r_01, swing1LimitDegrees=30, swing2LimitDegrees=15, twistLimitDegrees=2}
 {bone1Name=dyn_braid_r_03, bone2Name=dyn_braid_r_02, swing1LimitDegrees=40, swing2LimitDegrees=20, twistLimitDegrees=2}
 {bone1Name=dyn_braid_r_04, bone2Name=dyn_braid_r_03, swing1LimitDegrees=30, swing2LimitDegrees=30, twistLimitDegrees=2}
 {bone1Name=dyn_braid_r_05, bone2Name=dyn_braid_r_04, swing1LimitDegrees=60, swing2LimitDegrees=60, twistLimitDegrees=4}
 {bone1Name=dyn_hat_01, bone2Name=dyn_hat_MSTR, swing1LimitDegrees=60, swing2LimitDegrees=60, twistLimitDegrees=2}
 {bone1Name=dyn_hat_02, bone2Name=dyn_hat_01, swing1LimitDegrees=60, swing2LimitDegrees=60, twistLimitDegrees=2}
 {bone1Name=dyn_hat_03, bone2Name=dyn_hat_02, swing1LimitDegrees=60, swing2LimitDegrees=60, twistLimitDegrees=2}
 {bone1Name=dyn_hat_04, bone2Name=dyn_hat_03, swing1LimitDegrees=45, swing2LimitDegrees=45, twistLimitDegrees=4}
]
